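# Translucent unlit material for the WarmShade emote mesh, usable on Niagara mesh particles or skeletal meshes, with UseSKMesh bypassing particle color.
# Emissive layers a curve-atlas gradient, hue-shifted panning streaks masked by vertex red, and fiber noise in dissolving areas.
# Opacity combines blurred noise with an Appear-driven HeightLerp dissolve, using smoke noise sampled at floored world positions for a stepped look.
# UseOutlineEmissive blends to particle-colored emissive and line-texture opacity with an optional UV-based appear mask.
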Material /BRCosmetics/Effects/Fort_Effects/Emotes/WarmShade/Materials/M_Emote_WarmShade_Mesh
graph {
 properties {
  bAutomaticallySetUsageInEditor=false
  bOutputTranslucentVelocity=true
  bUsedWithMeshDeformer=true
  bUsedWithNiagaraMeshParticles=true
  bUsedWithSkeletalMesh=true
  blendMode=BLEND_Translucent
  shadingModel=MSM_Unlit
  usedShadingModels=Unlit
 }
 nodes [
  n1 = TextureSampleParameter2D(UVs=TextureCoordinate() {uTiling=0, vTiling=0.2}) {group=Color, parameterName=ColorNoise, samplerType=SAMPLERTYPE_Grayscale, texture=&LowResBlurredNoise}
  n2 = ScalarParameter() {group=Color, parameterName=ColorOffset}
  n3 = MaterialFunctionCall(
   Hue Shift Percentage=LinearInterpolate(
    A=ScalarParameter() {group=Color, parameterName=ColorRangeMin}
    B=ScalarParameter() {defaultValue=0.5, group=Color, parameterName=ColorRangeMax}
    Alpha=Add(A=@n1.R, B=@n2) {constB=0}
   ) {constB=0.5}
   Texture=VectorParameter() {parameterName=PannerColor, defaultValue=<rgba 0 7 1 1>}.RGB
  ) {materialFunction=&HueShift}
  n4 = TextureCoordinate()
  n5 = Multiply(
   A=Desaturation(None=Multiply(A=@n3.Result), Fraction=ScalarParameter() {defaultValue=-0.2, group=Color, parameterName=Des})
   B=TextureSample(UVs=Panner(Coordinate=@n4) {speedX=0.2, speedY=1}) {texture=/Game/Accessories/FORT_Backpacks/Backpack_M_MED_OceanBreeze/FX/T_OceanBreeze_SplineB_t}.R
  )
  n6 = VertexColor()
  n7 = ParticleColor()
  n8 = ScalarParameter() {parameterName=UseSKMesh}
  n9 = Multiply(
   A=TextureSample(UVs=TextureCoordinate()) {texture=/BRCosmetics/Weapons/FORT_Melee/Pickaxe_MadameMoth/FX/T_Fibre_Noise}.R
   B=Multiply(A=LinearInterpolate(A=@n7.RGB, Alpha=@n8) {constA=1})
  ) {constA=1}
  n10 = DynamicParameter() {defaultValue=<rgba 0 1 1 1>, paramNames [Appear,Param2,Param3,Param4]}
  n11 = Multiply(A=WorldPosition().XYZ, B=Constant() {r=0.2})
  n12 = MaterialFunctionCall(
   TextureObject=TextureObject() {texture=&T_Inky_Smoke_Tile}
   TextureSize=Constant() {r=20}
   WorldPosition=Add(A=Floor(None=@n11)) {constB=0}
  ) {materialFunction=&WorldAlignedTexture}
  n13 = Multiply(A=Saturate(None=LinearInterpolate(Alpha=@n12.XYZ Texture) {constB=2}))
  n14 = Saturate(None=Multiply(A=MaterialFunctionCall(Transition Phase=@n10.Appear, Height Texture=@n13) {materialFunction=&HeightLerp}.Results) {constB=3})
  n15 = OneMinus(None=@n14)
  n16 = CurveAtlasRowParameter(CurveTime=ComponentMask(None=TextureCoordinate() {coordinateIndex=1}) {g=true}) {
   atlas=&/CAtlas_Emote_WarmShade
   curve=&/CV_Emote_WarmShade_MainMesh
   defaultValue=4
   group=Color
   parameterName=Color
  }
  n17 = ParticleColor()
  n18 = Multiply(A=@n16, B=LinearInterpolate(A=@n17.RGB, Alpha=ScalarParameter() {parameterName=UseSKMesh}) {constA=1})
  n19 = Add(
   A=LinearInterpolate(B=Multiply(A=@n5, B=@n6.R), Alpha=ScalarParameter() {defaultValue=1, parameterName=UseVertexColor})
   B=Add(A=Multiply(A=@n9, B=@n15), B=Multiply(B=@n18) {constA=1})
  )
  n20 = ParticleColor()
  n21 = ScalarParameter() {parameterName=UseOutlineEmissive}
  MaterialFunctionCall(Emissive=Multiply(A=VectorParameter() {parameterName=MainColor, defaultValue=<rgba 1 1 1 1>}.RGB, B=LinearInterpolate(A=@n19, B=@n20.RGB, Alpha=@n21))) {materialFunction=&MF_FX_EmissiveAdjustments}
  n22 = LinearInterpolate(Alpha=DynamicParameter() {defaultValue=<rgba 0 1 1 1>, paramNames [Appear,Param2,Param3,Param4]}.Appear) {constA=-0.5}
  n23 = ComponentMask(None=TextureCoordinate() {coordinateIndex=1}) {r=true}
  n24 = Multiply(A=SmoothStep(Min=@n22, Max=Add(A=@n22) {constB=0.5}, Value=Sine(None=@n23) {period=2})) {constB=5}
  n25 = Multiply(A=@n20.A, B=LinearInterpolate(B=Saturate(None=@n24), Alpha=ScalarParameter() {parameterName=UseAppearMask}) {constA=1})
  n26 = Power(
   Base=TextureSample() {texture=/BRCosmetics/Animation/Props/Gadgets_WarmShade/Textures/T_Emote_Prop_Warmshade_E}.R
   Exp=ScalarParameter() {defaultValue=1, parameterName=LinePower}
  )
  n27 = TextureCoordinate() {uTiling=0.5, vTiling=0.5}
  n28 = Multiply(
   A=LinearInterpolate(A=@n17.A, Alpha=ScalarParameter() {parameterName=UseSKMesh}) {constA=1}
   B=LinearInterpolate(Alpha=TextureSample(UVs=@n27) {samplerType=SAMPLERTYPE_Grayscale, texture=&LowResBlurredNoise}.R) {constB=2}
  )
  Multiply(
   A=LinearInterpolate(A=Multiply(A=@n28, B=@n14), B=Multiply(A=@n25, B=@n26), Alpha=ScalarParameter() {parameterName=UseOutlineEmissive})
   B=ScalarParameter() {defaultValue=1, parameterName=MainAlpha}
  )
 ]
}
