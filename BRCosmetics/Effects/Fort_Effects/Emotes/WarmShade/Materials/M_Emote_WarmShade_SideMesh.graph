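# Unlit translucent material for Niagara mesh particles in the WarmShade emote, drawing a colored side mesh with hologram scanlines.
# Emissive adds a noise-distorted, panning streak texture with noise-varied hue shift to sine-derived edge highlights and a particle-tinted curve-atlas gradient.
# Opacity multiplies particle alpha, scanlines and an optional appear mask.
# Dynamic parameters drive the appear wipe and gradient desaturation, while a parameter selects UV0 or UV1 for color.
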Material /BRCosmetics/Effects/Fort_Effects/Emotes/WarmShade/Materials/M_Emote_WarmShade_SideMesh
graph {
 properties {
  bAutomaticallySetUsageInEditor=false
  bOutputTranslucentVelocity=true
  bUsedWithNiagaraMeshParticles=true
  blendMode=BLEND_Translucent
  shadingModel=MSM_Unlit
  usedShadingModels=Unlit
 }
 nodes [
  n1 = TextureSampleParameter2D(UVs=TextureCoordinate() {uTiling=0, vTiling=0.2}) {group=Color, parameterName=ColorNoise, samplerType=SAMPLERTYPE_Grayscale, texture=&LowResBlurredNoise}
  n2 = ScalarParameter() {group=Color, parameterName=ColorOffset}
  n3 = MaterialFunctionCall(
   Hue Shift Percentage=LinearInterpolate(
    A=ScalarParameter() {group=Color, parameterName=ColorRangeMin}
    B=ScalarParameter() {defaultValue=0.5, group=Color, parameterName=ColorRangeMax}
    Alpha=Add(A=@n1.R, B=@n2) {constB=0}
   ) {constB=0.5}
   Texture=Constant3Vector() {constant=<rgba 0 10 0 0>}
  ) {materialFunction=&HueShift}
  n4 = LinearInterpolate(
   A=TextureCoordinate() {coordinateIndex=1, uTiling=2, vTiling=0}
   B=TextureCoordinate() {uTiling=2, vTiling=0}
   Alpha=ScalarParameter() {parameterName=Use UV0}
  )
  n5 = Add(
   A=LinearInterpolate(A=TextureCoordinate() {coordinateIndex=1}, B=TextureCoordinate(), Alpha=ScalarParameter() {parameterName=Use UV0})
   B=Multiply(A=TextureSample(UVs=@n4) {samplerType=SAMPLERTYPE_Grayscale, texture=&LowResBlurredNoise}.R) {constB=0.3}
  ) {constB=0}
  n6 = Multiply(
   A=Desaturation(None=Multiply(A=@n3.Result), Fraction=ScalarParameter() {defaultValue=-0.2, group=Color, parameterName=Des})
   B=TextureSample(UVs=Panner(Coordinate=@n5) {speedX=0.2, speedY=1}) {texture=/Game/Accessories/FORT_Backpacks/Backpack_M_MED_OceanBreeze/FX/T_OceanBreeze_SplineB_t}.R
  )
  n7 = LinearInterpolate(A=TextureCoordinate() {coordinateIndex=1}, B=TextureCoordinate(), Alpha=ScalarParameter() {parameterName=Use UV0})
  n8 = Sine(None=ComponentMask(None=@n7) {r=true}) {period=2}
  n9 = Sine(None=ComponentMask(None=@n7) {g=true}) {period=2}
  n10 = Add(A=Power(Base=OneMinus(None=@n8)) {constExponent=20}, B=Power(Base=OneMinus(None=@n9)) {constExponent=30})
  n11 = CurveAtlasRowParameter(CurveTime=ComponentMask(None=@n7) {g=true}) {
   atlas=&/CAtlas_Emote_WarmShade
   curve=&/CV_Emote_WarmShade_SideMesh
   defaultValue=3
   group=Color
   parameterName=Color
  }
  n12 = DynamicParameter() {parameterIndex=1, defaultValue=<rgba 0 0 0 1>, paramNames [Des,none,none,none]}
  n13 = ParticleColor()
  n14 = Add(
   A=Multiply(A=Saturate(None=@n10), B=VectorParameter() {group=Color, parameterName=EdgeColor, defaultValue=<rgba 1 1 1 1>}.RGB)
   B=Multiply(A=Desaturation(None=@n11, Fraction=@n12.Des), B=@n13.RGB)
  )
  MaterialFunctionCall(Emissive=Multiply(A=Add(A=@n6, B=@n14))) {materialFunction=&MF_FX_EmissiveAdjustments}
  n15 = ScalarParameter() {defaultValue=200, group=Hologram, parameterName=Scanline Amount}
  n16 = ScalarParameter() {defaultValue=-0.1, group=Hologram, parameterName=ScanlineSize}
  n17 = ScalarParameter() {defaultValue=0.3, group=Hologram, parameterName=Scanlines Dark}
  n18 = ScalarParameter() {defaultValue=1, group=Hologram, parameterName=Scanlines Light}
  n19 = LinearInterpolate(Alpha=DynamicParameter() {paramNames [Appear,Param2,Param3,Param4]}.Appear) {constA=-0.5}
  n20 = ComponentMask(None=TextureCoordinate() {coordinateIndex=1}) {r=true}
  n21 = Multiply(A=SmoothStep(Min=@n19, Max=Add(A=@n19) {constB=0.5}, Value=Sine(None=@n20) {period=2})) {constB=5}
  Multiply(
   A=Multiply(
    A=@n13.A
    B=MaterialFunctionCall!(
     Number of Scanlines=@n15
     Scanline Speed
     "Scanlines Thickness (Bias)"=@n16
     Scanlines Dark Value=@n17
     Scanlines Light Value=@n18
     Block Opacity
     UVs
     Time
     Block Repeat
    ) {materialFunction=/Game/Effects/Fort_Effects/Effects/Atlas/Materials/MF_Atlas_Scanlines_01}.Result
   )
   B=LinearInterpolate(B=Saturate(None=@n21), Alpha=ScalarParameter() {parameterName=UseAppearMask}) {constA=1}
  )
  TextureCoordinate() {coordinateIndex=1}
 ]
}
