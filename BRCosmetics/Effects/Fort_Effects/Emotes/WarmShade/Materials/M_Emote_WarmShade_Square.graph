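# Translucent unlit Niagara mesh-particle material for the square shapes in the WarmShade emote.
# Color comes from a curve atlas sampled along offset instance bounds, with desaturation, particle tint, and a simple-color blend into emissive adjustments.
# Edge outlines use scale-compensated bounds masked by instance-space normals, or optionally UVs.
# Particle alpha times scanlines drives depth-faded opacity, and stepped, scrolling world-aligned noise builds a hue-shift glitch term.
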Material /BRCosmetics/Effects/Fort_Effects/Emotes/WarmShade/Materials/M_Emote_WarmShade_Square
graph {
 properties {
  bAutomaticallySetUsageInEditor=false
  bUsedWithNiagaraMeshParticles=true
  blendMode=BLEND_Translucent
  shadingModel=MSM_Unlit
  usedShadingModels=Unlit
 }
 nodes [
  ComponentMask(None=Subtract(A=WorldPosition().XYZ, B=ParticlePositionWS())) {b=true}
  Constant3Vector() {constant=<rgba 1 0.4754 0.5353 1>}
  Constant3Vector() {constant=<rgba 20 1.7864 6.3754 1>}
  Constant3Vector() {constant=<rgba 1 1 1 0>}
  n1 = ParticleColor()
  n2 = ScalarParameter() {defaultValue=200, parameterName=Scanline Amount}
  n3 = ScalarParameter() {defaultValue=-0.1, parameterName=ScanlineSize}
  n4 = ScalarParameter() {defaultValue=0.5, parameterName=Scanlines Dark}
  n5 = ScalarParameter() {defaultValue=1, parameterName=Scanlines Light}
  DepthFade(
   Opacity=Multiply(
    A=@n1.A
    B=MaterialFunctionCall!(
     Number of Scanlines=@n2
     Scanline Speed
     "Scanlines Thickness (Bias)"=@n3
     Scanlines Dark Value=@n4
     Scanlines Light Value=@n5
     Block Opacity
     UVs
     Time
     Block Repeat
    ) {materialFunction=/Game/Effects/Fort_Effects/Effects/Atlas/Materials/MF_Atlas_Scanlines_01}.Result
   )
   FadeDistance=ScalarParameter() {parameterName=FadeDistance}
  )
  n6 = Divide(A=MaterialFunctionCall() {materialFunction=/Game/Effects/Niagara/MasterMaterials/Functions/InstanceBounds_0-1_UVW}.B) {constB=3}
  n7 = ScalarParameter() {parameterName=Offset}
  n8 = CurveAtlasRowParameter(CurveTime=Multiply(A=Add(A=@n6, B=@n7) {constB=0.25})) {
   atlas=&/CAtlas_Emote_WarmShade
   curve=&/CV_Emote_WarmShade_SquareColor
   parameterName=Color
  }
  n9 = Desaturation(None=@n8, Fraction=DynamicParameter() {parameterIndex=1, defaultValue=<rgba 0 0 0 1>, paramNames [Des,none,none,none]}.Des)
  n10 = NamedRerouteUsage()
  n11 = Multiply(A=Add(A=Multiply(A=@n9), B=LinearInterpolate(Alpha=@n10) {constB=2.5}))
  n12 = Multiply(A=@n11, B=@n1.RGB)
  n13 = NamedRerouteUsage()
  MaterialFunctionCall(Emissive=LinearInterpolate(A=Add(B=@n12), B=Add(A=@n1.RGB, B=@n13), Alpha=ScalarParameter() {parameterName=UseSimpleParticleColor})) {materialFunction=&MF_FX_EmissiveAdjustments}
  n14 = Time()
  n15 = Constant3Vector() {constant=<rgba 0 0 -10 0>}
  n16 = Constant() {r=0.3}
  n17 = Floor(None=Multiply(A=WorldPosition().XYZ, B=@n16))
  n18 = MaterialFunctionCall(
   TextureObject=TextureObject() {samplerType=SAMPLERTYPE_Grayscale, texture=&LowResBlurredNoise}
   TextureSize=Constant() {r=50}
   WorldPosition=Add(A=Multiply(A=@n14, B=@n15), B=Divide(A=@n17, B=@n16))
  ) {materialFunction=/Game/Weapons/WeaponSkins/Wraps/MaterialFunctions/WorldAlignedTexture}
  n19 = LinearInterpolate(Alpha=Saturate(None=LinearInterpolate(Alpha=@n18.XYZ Texture) {constA=-1, constB=2})) {constA=-0.5, constB=20}
  n20 = LinearInterpolate(Alpha=@n19) {constA=-0.2, constB=-0.1}
  Min(
   A=@n11
   B=Multiply(
    A=MaterialFunctionCall(Hue Shift Percentage=@n20) {materialFunction=&HueShift}.Result
    B=Saturate(None=@n19)
   )
  )
  Multiply(A=@n18.XYZ Texture)
  n21 = VectorParameter() {parameterName=EdgeLineColor, defaultValue=<rgba 1 1 1 1>}
  n22 = MaterialFunctionCall() {materialFunction=&InstanceBounds_0-1_UVW}
  n23 = DynamicParameter() {defaultValue=<rgba 1 1 15 1>, paramNames [ScaleX,ScaleY,ScaleZ,Param4]}
  n24 = OneMinus(None=@n22.G)
  n25 = ScalarParameter() {defaultValue=0.05, parameterName=BandWidth}
  n26 = Divide(A=Min(A=Multiply(A=@n22.G, B=@n23.ScaleY), B=Multiply(A=@n24, B=@n23.ScaleY)), B=@n25)
  n27 = ScalarParameter() {defaultValue=2, parameterName=Grad}
  n28 = Power(Base=OneMinus(None=Saturate(None=@n26)), Exp=@n27)
  n29 = Normalize(VectorInput=Transform(None=VertexNormalWS()) {transformSourceType=TRANSFORMSOURCE_World, transformType=TRANSFORM_Instance})
  n30 = Abs(None=DotProduct(A=@n29, B=Constant3Vector() {constant=<rgba 0 1 0 0>}))
  n31 = OneMinus(None=Ceil(None=Subtract(A=@n30) {constB=0.01}))
  n32 = OneMinus(None=@n22.R)
  n33 = Divide(A=Min(A=Multiply(A=@n23.ScaleX, B=@n22.R), B=Multiply(A=@n32, B=@n23.ScaleX)), B=@n25)
  n34 = Power(Base=OneMinus(None=Saturate(None=@n33)), Exp=@n27)
  n35 = Abs(None=DotProduct(A=@n29, B=Constant3Vector() {constant=<rgba 1 0 0 0>}))
  n36 = OneMinus(None=Ceil(None=Subtract(A=@n35) {constB=0.01}))
  n37 = Min(A=Multiply(A=@n22.B, B=@n23.ScaleZ), B=Multiply(A=OneMinus(None=@n22.B), B=@n23.ScaleZ))
  n38 = OneMinus(None=Saturate(None=Divide(A=@n37, B=@n25)))
  n39 = DotProduct(A=@n29, B=Constant3Vector() {constant=<rgba 0 0 1 0>})
  n40 = Ceil(None=Subtract(A=Abs(None=@n39)) {constB=0.01})
  n41 = Max(A=Max(A=Multiply(A=@n28, B=@n31), B=Multiply(A=@n34, B=@n36)), B=Multiply(A=Power(Base=@n38, Exp=@n27), B=OneMinus(None=@n40)))
  n42 = TextureCoordinate()
  n43 = OneMinus(None=Sine(None=ComponentMask(None=@n42) {r=true}) {period=2})
  n44 = ScalarParameter() {defaultValue=20, parameterName=EdgeWidth}
  n45 = OneMinus(None=Sine(None=ComponentMask(None=@n42) {g=true}) {period=2})
  n46 = Saturate(None=Add(A=Power(Base=@n43, Exp=@n44) {constExponent=20}, B=Power(Base=@n45, Exp=@n44) {constExponent=20}))
  NamedRerouteDeclaration(None=LinearInterpolate(A=Multiply(A=@n21.RGB, B=@n41), B=Multiply(A=@n21.RGB, B=@n46), Alpha=ScalarParameter() {parameterName=UseUVOutLine})) {name=EdgeLine, nodeColor=<rgba 0 0.25 1 1>}
  ScalarParameter() {parameterName=Des}
  ScalarParameter() {defaultValue=1, parameterName=UseGlitch}
  TextureSample() {texture=&T_Inky_Smoke_Tile}
  WorldPosition()
 ]
}
