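# Unlit translucent material for Niagara mesh particles in the WarmShade emote, producing an animated blocky square glitch pattern.
# Floor operations quantize world position and UVs into cells that scroll over a blurred noise texture, with UseUVPanner blending between world-aligned and UV-panned samples.
# Two chained remaps sharpen the noise into high-contrast squares.
# Particle color tints the result through an emissive adjustment function, and particle alpha scales it for opacity.
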Material /BRCosmetics/Effects/Fort_Effects/Emotes/WarmShade/Materials/M_Emote_WarmShade_SquareGlitch
graph {
 properties {
  bAutomaticallySetUsageInEditor=false
  bOutputTranslucentVelocity=true
  bUsedWithNiagaraMeshParticles=true
  blendMode=BLEND_Translucent
  shadingModel=MSM_Unlit
  usedShadingModels=Unlit
 }
 nodes [
  n1 = Time()
  n2 = VectorParameter() {parameterName=PannerDirection, defaultValue=<rgba 0 0 -10 1>}
  n3 = Constant() {r=0.3}
  n4 = Floor(None=Multiply(A=WorldPosition().XYZ, B=@n3))
  n5 = MaterialFunctionCall(
   TextureObject=TextureObject() {samplerType=SAMPLERTYPE_Grayscale, texture=&LowResBlurredNoise}
   TextureSize=Constant() {r=50}
   WorldPosition=Add(A=Multiply(A=@n1, B=@n2.RGB), B=Divide(A=@n4, B=@n3))
  ) {materialFunction=/Game/Weapons/WeaponSkins/Wraps/MaterialFunctions/WorldAlignedTexture}
  n6 = TextureCoordinate()
  n7 = ScalarParameter() {defaultValue=1, parameterName=UVScale}
  n8 = ScalarParameter() {defaultValue=10, parameterName=PixelScale}
  n9 = Floor(None=Multiply(A=Multiply(A=@n6, B=@n7), B=@n8))
  n10 = VectorParameter() {parameterName=UVoffset, channelNames {a="", b="", g=V Panner, r=U Panner}}
  n11 = TextureSample(UVs=Panner(Coordinate=Divide(A=@n9, B=@n8), Speed=ComponentMask(None=@n10) {g=true, r=true})) {samplerType=SAMPLERTYPE_Grayscale, texture=&LowResBlurredNoise}
  n12 = ScalarParameter() {parameterName=UseUVPanner}
  n13 = Saturate(None=LinearInterpolate(Alpha=LinearInterpolate(A=@n5.XYZ Texture, B=@n11.R, Alpha=@n12)) {constA=-0.5})
  n14 = Multiply(B=Saturate(None=LinearInterpolate(Alpha=@n13) {constA=-0.5, constB=5})) {constA=1}
  n15 = ParticleColor()
  MaterialFunctionCall(Emissive=Multiply(A=@n14, B=@n15.RGB)) {materialFunction=&MF_FX_EmissiveAdjustments}
  Multiply(A=@n14, B=@n15.A)
 ]
}
