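# Translucent unlit material for Niagara mesh particles, giving the tech mesh in the WarmShade emote a holographic scanline look.
# Emissive adds a white UV-border glow, built from inverted sine masks raised to powers, to a U-sampled curve-atlas gradient tinted by particle color, then runs through an emissive adjustment function.
# An opacity term multiplies particle alpha, a linear fade along U and a parameterized scanline function.
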Material /BRCosmetics/Effects/Fort_Effects/Emotes/WarmShade/Materials/M_Emote_WarmShade_TechMesh
graph {
 properties {
  bAutomaticallySetUsageInEditor=false
  bUsedWithNiagaraMeshParticles=true
  blendMode=BLEND_Translucent
  shadingModel=MSM_Unlit
  usedShadingModels=Unlit
 }
 nodes [
  n1 = TextureCoordinate()
  n2 = OneMinus(None=Sine(None=ComponentMask(None=@n1) {r=true}) {period=2})
  n3 = TextureCoordinate()
  n4 = OneMinus(None=Sine(None=ComponentMask(None=@n3) {g=true}) {period=2})
  n5 = Saturate(None=Add(A=Power(Base=@n2) {constExponent=40}, B=Power(Base=@n4) {constExponent=10}))
  n6 = CurveAtlasRowParameter(CurveTime=ComponentMask(None=TextureCoordinate()) {r=true}) {
   atlas=&/CAtlas_Emote_WarmShade
   curve=&/CV_Emote_WarmShade_TechMesh
   defaultValue=2
   parameterName=Color
  }
  n7 = ParticleColor()
  MaterialFunctionCall(Emissive=Add(A=Multiply(B=@n5) {constA=1}, B=Multiply(A=@n6, B=@n7.RGB))) {materialFunction=&MF_FX_EmissiveAdjustments}
  n8 = OneMinus(None=ComponentMask(None=TextureCoordinate()) {r=true})
  Multiply(
   A=Multiply(A=@n7.A, B=LinearInterpolate(Alpha=@n8) {constA=0.5})
   B=MaterialFunctionCall!(
    Number of Scanlines=ScalarParameter() {defaultValue=150, parameterName=Scanline Amount}
    Scanline Speed
    "Scanlines Thickness (Bias)"=ScalarParameter() {defaultValue=-0.1, parameterName=ScanlineSize}
    Scanlines Dark Value=ScalarParameter() {defaultValue=0.5, parameterName=Scanlines Dark}
    Scanlines Light Value=ScalarParameter() {defaultValue=1, parameterName=Scanlines Light}
    Block Opacity
    UVs
    Time
    Block Repeat
   ) {materialFunction=/Game/Effects/Fort_Effects/Effects/Atlas/Materials/MF_Atlas_Scanlines_01}.Result
  )
 ]
}
