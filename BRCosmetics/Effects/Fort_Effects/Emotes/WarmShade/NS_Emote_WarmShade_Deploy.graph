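# Niagara system for the WarmShade emote deploy effect, a glitchy, multicolored appearance of a mesh.
# MainMesh, Glitch and Outline spawn mesh particles with curve-driven color or alpha and a material parameter animated from one to zero.
# BurstRay adds bright color-cycling mesh streaks, while two partitioned Glow sprites fade in on either side.
# Two Glitch_sprite mesh emitters run ShapeLocation in particle update, re-placing flickering, color-cycling pieces within a box for jitter.
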NiagaraSystem /BRCosmetics/Effects/Fort_Effects/Emotes/WarmShade/NS_Emote_WarmShade_Deploy
dependencies {
 usedDataInterfaces [
  &NiagaraDataInterfaceColorCurve
  &NiagaraDataInterfaceCurve
  &NiagaraDataInterfaceVectorCurve
  &NiagaraDataInterfaceArrayFloat3
 ]
 usedDynamicInputs [&ConvertVectorToPosition]
 usedModules [&SystemState]
}
emitters {
 BurstRay {
  emitterUpdateScript [
   &EmitterState(Life Cycle Mode=::NewEnumerator1, Loop Behavior=::NewEnumerator1, Loop Delay=1, Loop Duration=0.15, UseLoopDelay=true)
   &SpawnRate(SpawnRate=10, Use Spawn Probability=false)
   &SpawnBurst_Instantaneous(Spawn Count=30, Use Spawn Probability=false, UseLoopCountLimit=false)
  ]
  particleSpawnScript [
   &InitializeParticle(
    Color=<rgba 1 0.5424 0.6325 1>
    Lifetime Max=0.7
    Lifetime Min=0.3
    Lifetime Mode=::NewEnumerator1
    Mesh Scale=&Multiply_VectorByFloat(Vector=&RandomRangeVector(Minimum=<0.3 0.3 0.5>, Maximum=<0.6 0.6 1>), Float=0.1)
    Mesh Scale Mode=::NewEnumerator4
    UsePositionOffset=false
    Write Material Random=true
   )
   &ShapeLocation(Box Size=<150 60 50>, MaskSpawnGroup=false, Override Seed=false, Shape Primitive=::NewEnumerator2, Surface Only=false, UseOwnerScale=true)
  ]
  particleUpdateScript [
   &ParticleState()
   &ScaleMeshSize(Scale Factor=&VectorFromCurve(VectorCurve=DataInterface() {propertyValues {
       XCurve {keys [<at 0 1 RCIM_Cubic>,<at 1 1>]}
       YCurve {keys [<at 0 1 RCIM_Cubic>,<at 1 1>]}
       ZCurve {keys [{arriveTangent=0.9527, interpMode=RCIM_Cubic, time=0, value=0.5}, <at 1 1.5>]}
      }}))
   &DynamicMaterialParameters(
    Index 0 Param 1=1
    Index 0 Param 2=1
    Index 0 Param 3=1
    Index 0 Param 4=1
    Module.Index0Param0Write=true
    Module.Index0Param1Write=true
    Module.Index0Param2Write=true
    Module.Index0Param3Write=true
   )
   &CameraOffset(Camera Offset Amount=15)
   &ScaleColor(
    Scale Alpha=&FloatFromCurve(FloatCurve=DataInterface() {propertyValues {Curve {keys [<at 0 1 RCIM_Cubic>, {arriveTangent=-1, interpMode=RCIM_Cubic, leaveTangent=-1, time=0.6047, value=0.4996}, <at 1 0>]}}})
    Scale RGB=&Multiply_Vector(
     A=&VectorFromFloat(Value=10)
     B=&MakeVectorFromLinearColorRGB(Color=&ColorFromCurve(ColorCurve=DataInterface() {propertyValues {
         AlphaCurve {keys [<at 0 1 RCIM_Cubic>]}
         BlueCurve {keys [<at 0.0176 0.8136>,<at 0.1843 0.0889>,<at 0.4216 1>,<at 0.548 0.1268>,<at 0.7273 0.2395>,<at 1 0.0671 RCIM_Cubic>]}
         GreenCurve {keys [<at 0.0176 0.0856>,<at 0.1843 0.0889>,<at 0.4216 0.5705>,<at 0.548 0.0371>,<at 0.7273 0.2395>,<at 1 0.4357 RCIM_Cubic>]}
         RedCurve {keys [<at 0.0176 1>,<at 0.1843 1>,<at 0.4216 0>,<at 0.548 1>,<at 0.7273 1>,<at 1 1 RCIM_Cubic>]}
        }}))
    )
    ScaleA=true
    ScaleRGB=true
   )
   &FN_Mesh_Attributes(
    Color=&MakeLinearColorFromVectorAndFloat(
     "Vector (RGB)"=&Multiply_VectorByFloat(
      Vector=&MakeVectorFromLinearColorRGB(Color=&ColorFromCurve(ColorCurve=DataInterface() {propertyValues {
          AlphaCurve {keys [<at 0 1 RCIM_Cubic>]}
          BlueCurve {keys [<at 0.0176 0.8136>,<at 0.1843 0.0889>,<at 0.4216 1>,<at 0.548 0.1268>,<at 0.7273 0.2395>,<at 1 0.0671 RCIM_Cubic>]}
          GreenCurve {keys [<at 0.0176 0.0856>,<at 0.1843 0.0889>,<at 0.4216 0.5705>,<at 0.548 0.0371>,<at 0.7273 0.2395>,<at 1 0.4357 RCIM_Cubic>]}
          RedCurve {keys [<at 0.0176 1>,<at 0.1843 1>,<at 0.4216 0>,<at 0.548 1>,<at 0.7273 1>,<at 1 1 RCIM_Cubic>]}
         }}))
      Float=10
     )
     "Float (Alpha)"=&FloatFromCurve(
      FloatCurve=DataInterface() {propertyValues {Curve {keys [<at 0 1 RCIM_Cubic>, {arriveTangent=-1, interpMode=RCIM_Cubic, leaveTangent=-1, time=0.6047, value=0.4996}, <at 1 0>]}}}
      Scale Curve=10
     )
    )
    Create Color Attribute=true
    Create Scale Attribute=true
    Scale=<1 1 0.1>
    Write to Module Namespace=true
   )
  ]
  rendererClasses [&NiagaraMeshRendererProperties]
  renderers [
   {rendererClass=&NiagaraMeshRendererProperties, rendererIndex=0}
   {rendererClass=&NiagaraMeshRendererProperties, rendererIndex=1}
  ]
 }
 Glitch {
  emitterUpdateScript [
   &EmitterState(
    Inactive Response=::NewEnumerator1
    Life Cycle Mode=::NewEnumerator1
    Loop Behavior=::NewEnumerator1
    Loop Delay=0.3
    Loop Duration=3
    UseLoopDelay=true
   )
   &SpawnBurst_Instantaneous(Use Spawn Probability=false, UseLoopCountLimit=false)
  ]
  particleSpawnScript [
   &InitializeParticle(
    Color=<rgba 1.5 1.5 1.5 1>
    Lifetime=2
    Mesh Scale=&Multiply_VectorByFloat(Vector=<1 1 1.2>, Float=1)
    Mesh Scale Mode=::NewEnumerator4
    UsePositionOffset=false
    Write Material Random=true
   )
   &CameraOffset(Camera Offset Amount=1)
  ]
  particleUpdateScript [
   &ParticleState()
   &ScaleColor(
    Scale Alpha=&Multiply_Float(A=2, B=1)
    Scale RGB=&Multiply_VectorByFloat(
     Vector=&MakeVectorFromLinearColorRGB(Color=&ColorFromCurve(ColorCurve=DataInterface() {propertyValues {
         AlphaCurve {keys []}
         BlueCurve {keys [<at 0.0429 1>, {arriveTangent=-8.5403, interpMode=RCIM_Cubic, time=0.1129, value=0.3696}, <at 0.1696 0>, <at 0.242 0.2021 RCIM_Cubic>]}
         GreenCurve {keys [<at 0.0429 0.5072>, {arriveTangent=-3.9343, interpMode=RCIM_Cubic, time=0.1129, value=1}, <at 0.1696 0.0465>, <at 0.242 0.1323 RCIM_Cubic>]}
         RedCurve {keys [<at 0.0429 0>, {arriveTangent=8.5403, interpMode=RCIM_Cubic, time=0.1129, value=0}, <at 0.1696 1>, <at 0.242 1 RCIM_Cubic>]}
        }}))
     Float=1
    )
    ScaleA=true
    ScaleRGB=true
   )
   &DynamicMaterialParameters(
    Index 0 Param 1=&FloatFromCurve(FloatCurve=DataInterface() {propertyValues {Curve {keys [
         {arriveTangent=0.6917, interpMode=RCIM_Cubic, leaveTangent=0.6917, tangentMode=RCTM_User, time=0, value=1}
         {arriveTangent=-4.5607, interpMode=RCIM_Cubic, leaveTangent=-4.5607, tangentMode=RCTM_User, time=0.4, value=0}
        ]}}})
    Module.Index0Param0Write=true
    Module.Index0Param1Write=true
    Module.Index0Param2Write=true
    Module.Index0Param3Write=true
   )
   &ScaleColor(
    Scale RGB=&Multiply_VectorByFloat(
     Vector=&MakeVector(
      X=&FloatFromCurve(FloatCurve=DataInterface() {propertyValues {Curve {keys [<at 0 1>,<at 0.4 1>,<at 0.5 1>,<at 0.7 1>]}}})
      Y=&FloatFromCurve(FloatCurve=DataInterface() {propertyValues {Curve {keys [<at 0 1>,<at 0.4 1>,<at 0.5 1>,<at 0.7 1>]}}})
      Z=&FloatFromCurve(FloatCurve=DataInterface() {propertyValues {Curve {keys [<at 0 1>,<at 0.4 1>,<at 0.5 1>,<at 0.7 1>]}}})
     )
     Float=1
    )
    ScaleA=true
    ScaleRGB=true
   ) {module=ScaleColor001}
  ]
  rendererClasses [&NiagaraMeshRendererProperties]
  renderers [{rendererClass=&NiagaraMeshRendererProperties, rendererIndex=0}]
 }
 Glitch_sprite {
  emitterUpdateScript [
   &EmitterState(Life Cycle Mode=::NewEnumerator1, Loop Behavior=::NewEnumerator1, Loop Duration=0.6, UseLoopDelay=false)
   &SpawnRate(SpawnRate=50, Use Spawn Probability=false)
  ]
  particleSpawnScript [
   &InitializeParticle(
    Color=<rgba 5 1.3021 1.3842 3>
    Lifetime Max=0.4
    Lifetime Min=0.3
    Lifetime Mode=::NewEnumerator1
    Mesh Scale Max=&MakeCustomVectorFromBool(Bool=&RandomBool(), True Vector=<2 4 1>, False Vector=<2 2 8>)
    Mesh Scale Min=<1 1 3>
    Mesh Scale Mode=::NewEnumerator2
    UsePositionOffset=true
    Write Material Random=true
   )
   &CameraOffset()
  ]
  particleUpdateScript [
   &ShapeLocation(
    Box Size=<200 80 30>
    MaskSpawnGroup=false
    Offset=<0 -10 0>
    Offset Mode=::NewEnumerator0
    Override Seed=false
    Shape Origin=&VectorFromFloat(Value=0)
    Shape Primitive=::NewEnumerator2
    Surface Only=false
    UseOwnerScale=true
   )
   &ParticleState()
   &ScaleMeshSize(Scale Factor=&Multiply_VectorByFloat(
     Vector=<1 1 1>
     Float=&Multiply_Float(
      A=0.15
      B=&FloatFromCurve(
       FloatCurve=DataInterface() {propertyValues {Curve {keys [
           {arriveTangent=4.1785, time=0, value=1}
           <at 0.5 1.1>
           {arriveTangent=-0.0405, interpMode=RCIM_Cubic, leaveTangent=-0.0405, tangentMode=RCTM_User, time=1, value=1.5}
          ]}}}
       Scale Curve=2
      )
     )
    ))
   &Drag(Drag=2, Use Linear Drag=true, Use Rotational Drag=true)
   &SolveForcesAndVelocity(Clamp Velocity=false, Limit Acceleration=false)
   &DynamicMaterialParameters(Module.Param0WriteEnabled=true, Module.Param1WriteEnabled=true, Module.Param2WriteEnabled=true, Module.Param3WriteEnabled=true)
   &ScaleColor(
    Scale Alpha=&Multiply_Float(
     A=&FloatFromCurve(
      FloatCurve=DataInterface() {propertyValues {Curve {keys [
          {arriveTangent=9.0725, interpMode=RCIM_Cubic, leaveTangent=9.0725, tangentMode=RCTM_User, time=0, value=0}
          {arriveTangent=-2.5249, interpMode=RCIM_Cubic, leaveTangent=-2.5249, tangentMode=RCTM_User, time=0.2, value=0.9847}
          {arriveTangent=-0.0405, interpMode=RCIM_Cubic, leaveTangent=-0.0405, tangentMode=RCTM_User, time=1, value=0}
         ]}}}
      Scale Curve=&RandomRangeFloat(Minimum=0.2, Maximum=1)
     )
     B=2
    )
    Scale RGB=&Multiply_VectorByFloat(
     Vector=&MakeVectorFromLinearColorRGB(Color=&ColorFromCurve(ColorCurve=DataInterface() {propertyValues {
         AlphaCurve {keys []}
         BlueCurve {keys [
           <at 0 5 RCIM_Cubic>
           <at 0.1326 0>
           <at 0.2869 1.4885>
           <at 0.4134 5>
           <at 0.5108 0.2805>
           <at 0.6236 5>
           {arriveTangent=-25.2941, interpMode=RCIM_Cubic, leaveTangent=-25.2941, time=0.7107, value=0}
           <at 0.8097 0.2928>
          ]}
         GreenCurve {keys [
           <at 0 2.5535 RCIM_Cubic>
           <at 0.1326 0>
           <at 0.2869 5>
           <at 0.4134 0>
           <at 0.5108 0>
           <at 0.6236 4.8554>
           {arriveTangent=-26.0903, interpMode=RCIM_Cubic, leaveTangent=-26.0903, time=0.7107, value=0}
           <at 0.8097 0>
          ]}
         RedCurve {keys [
           <at 0 0 RCIM_Cubic>
           <at 0.1326 1>
           <at 0.2869 0>
           <at 0.4134 0.4394>
           <at 0.5108 5>
           <at 0.6236 0>
           {arriveTangent=26.8676, interpMode=RCIM_Cubic, leaveTangent=26.8676, time=0.7107, value=1}
           <at 0.8097 5>
          ]}
        }}))
     Float=&FloatFromCurve(
      FloatCurve=DataInterface() {propertyValues {Curve {keys [{arriveTangent=-2.0684, interpMode=RCIM_Cubic, leaveTangent=-2.0684, tangentMode=RCTM_User, time=0, value=2}, <at 1 0>]}}}
      Scale Curve=&Multiply_Float(A=0.5, B=0.5)
     )
    )
    ScaleA=true
    ScaleRGB=true
   )
  ]
  rendererClasses [&NiagaraMeshRendererProperties]
  renderers [{rendererClass=&NiagaraMeshRendererProperties, rendererIndex=0}]
 }
 Glitch_sprite001 {
  emitterUpdateScript [
   &EmitterState(Life Cycle Mode=::NewEnumerator1, Loop Behavior=::NewEnumerator1, Loop Delay=0.5, Loop Duration=1, UseLoopDelay=true)
   &SpawnRate(SpawnRate=60, Use Spawn Probability=false)
  ]
  particleSpawnScript [
   &InitializeParticle(
    Color=<rgba 1 0.2604 0.2768 1>
    Lifetime Max=0.4
    Lifetime Min=0.3
    Lifetime Mode=::NewEnumerator1
    Mesh Scale Max=&MakeCustomVectorFromBool(Bool=&RandomBool(), True Vector=<2 5 1>, False Vector=<2 2 10>)
    Mesh Scale Min=<1 1 3>
    Mesh Scale Mode=::NewEnumerator2
    UsePositionOffset=true
    Write Material Random=true
   )
   &CameraOffset()
   &InitialMeshOrientation(Mesh Orientation Mode=::NewEnumerator7, Rotation=<0 0.25 0>, Use Rotation Vector=true)
  ]
  particleUpdateScript [
   &ShapeLocation(
    Box Size=&MakeVector(X=200, Y=80, Z=30)
    MaskSpawnGroup=false
    Offset=<0 -10 0>
    Offset Mode=::NewEnumerator0
    Override Seed=false
    Shape Origin=&VectorFromFloat(Value=0)
    Shape Primitive=::NewEnumerator2
    Surface Only=false
    UseOwnerScale=true
   ) {module=ShapeLocation001}
   &ParticleState()
   &ScaleMeshSize(Scale Factor=&Multiply_VectorByFloat(
     Vector=<1 1 1>
     Float=&Multiply_Float(
      A=0.15
      B=&FloatFromCurve(
       FloatCurve=DataInterface() {propertyValues {Curve {keys [
           {arriveTangent=4.1785, time=0, value=1}
           <at 0.5 1.1>
           {arriveTangent=-0.0405, interpMode=RCIM_Cubic, leaveTangent=-0.0405, tangentMode=RCTM_User, time=1, value=1.5}
          ]}}}
       Scale Curve=0.7
      )
     )
    ))
   &Drag(Drag=2, Use Linear Drag=true, Use Rotational Drag=true)
   &SolveForcesAndVelocity(Clamp Velocity=false, Limit Acceleration=false)
   &DynamicMaterialParameters(Module.Param0WriteEnabled=true, Module.Param1WriteEnabled=true, Module.Param2WriteEnabled=true, Module.Param3WriteEnabled=true)
   &ScaleColor(
    Scale Alpha=&Multiply_Float(
     A=&FloatFromCurve(
      FloatCurve=DataInterface() {propertyValues {Curve {keys [
          {arriveTangent=9.0725, interpMode=RCIM_Cubic, leaveTangent=9.0725, tangentMode=RCTM_User, time=0, value=0}
          {arriveTangent=-2.5249, interpMode=RCIM_Cubic, leaveTangent=-2.5249, tangentMode=RCTM_User, time=0.2, value=0.9847}
          {arriveTangent=-0.0405, interpMode=RCIM_Cubic, leaveTangent=-0.0405, tangentMode=RCTM_User, time=1, value=0}
         ]}}}
      Scale Curve=&RandomRangeFloat(Minimum=0.2, Maximum=1)
     )
     B=1
    )
    Scale RGB=&Multiply_VectorByFloat(
     Vector=&MakeVectorFromLinearColorRGB(Color=&ColorFromCurve(ColorCurve=DataInterface() {propertyValues {
         AlphaCurve {keys []}
         BlueCurve {keys [
           <at 0 5 RCIM_Cubic>
           <at 0.1326 0>
           <at 0.2869 0.4688>
           <at 0.4134 5>
           <at 0.5108 0.2805>
           <at 0.6236 5>
           {arriveTangent=-25.2941, interpMode=RCIM_Cubic, leaveTangent=-25.2941, time=0.7107, value=0}
           <at 0.8097 0.2928>
          ]}
         GreenCurve {keys [
           <at 0 2.5535 RCIM_Cubic>
           <at 0.1326 0>
           <at 0.2869 0.512>
           <at 0.4134 0>
           <at 0.5108 0>
           <at 0.6236 4.8554>
           {arriveTangent=-26.0903, interpMode=RCIM_Cubic, leaveTangent=-26.0903, time=0.7107, value=0}
           <at 0.8097 0>
          ]}
         RedCurve {keys [
           <at 0 0 RCIM_Cubic>
           <at 0.1326 1>
           <at 0.2869 5>
           <at 0.4134 0.4394>
           <at 0.5108 5>
           <at 0.6236 0>
           {arriveTangent=26.8676, interpMode=RCIM_Cubic, leaveTangent=26.8676, time=0.7107, value=1}
           <at 0.8097 5>
          ]}
        }}))
     Float=&FloatFromCurve(
      FloatCurve=DataInterface() {propertyValues {Curve {keys [{arriveTangent=-2.0684, interpMode=RCIM_Cubic, leaveTangent=-2.0684, tangentMode=RCTM_User, time=0, value=2}, <at 1 0>]}}}
      Scale Curve=&Multiply_Float(A=0.5, B=0.5)
     )
    )
    ScaleA=true
    ScaleRGB=true
   )
  ]
  rendererClasses [&NiagaraMeshRendererProperties]
  renderers [{rendererClass=&NiagaraMeshRendererProperties, rendererIndex=0}]
 }
 Glow {
  emitterUpdateScript [
   &EmitterState(Life Cycle Mode=::NewEnumerator1, Loop Behavior=::NewEnumerator1, Loop Duration=1, UseLoopDelay=false)
   &SpawnBurst_Instantaneous(Spawn Count=2, Use Spawn Probability=false, UseLoopCountLimit=false)
  ]
  particleSpawnScript [
   &PartitionParticles()
   &InitializeParticle(
    Color=<rgba 0.4844 0.0097 0.0492 0.3>
    Position Offset=&SelectVectorFromArray(
     Direct Array Index=$Particles.PartitionParticles.Partition
     Vector Selection Array=DataInterface() {propertyValues {FloatData [<-40 -15 0>, <40 -15 0>]}}
    )
    Sprite Size Mode=::NewEnumerator3
    Uniform Sprite Size=350
    UsePositionOffset=true
    Write Material Random=true
   )
   &SpriteFacingAndAlignment(Sprite Alignment=<1 0 0>, Sprite Facing=<0 0 1>, WriteAlignment=true, WriteFacing=true)
   &DynamicMaterialParameters(
    Index 0 Param 1=3
    Index 0 Param 2=10
    Module.Index0Param0Write=true
    Module.Index0Param1Write=true
    Module.Index0Param2Write=true
    Module.Index0Param3Write=true
   )
  ]
  particleUpdateScript [
   &ParticleState()
   &SolveForcesAndVelocity(Clamp Velocity=false, Limit Acceleration=false)
   &ScaleColor(
    Scale Alpha=&FloatFromCurve(FloatCurve=DataInterface() {propertyValues {Curve {keys [{tangentMode=RCTM_User, time=0, value=0}, {tangentMode=RCTM_User, time=0.5, value=1}]}}})
    ScaleA=true
    ScaleRGB=true
   )
  ]
  rendererClasses [&NiagaraSpriteRendererProperties]
  renderers [{rendererClass=&NiagaraSpriteRendererProperties}]
 }
 MainMesh {
  emitterUpdateScript [
   &EmitterState(
    Inactive Response=::NewEnumerator1
    Life Cycle Mode=::NewEnumerator1
    Loop Behavior=::NewEnumerator1
    Loop Delay=0.3
    Loop Duration=3
    UseLoopDelay=true
   )
   &SpawnBurst_Instantaneous(Use Spawn Probability=false, UseLoopCountLimit=false)
  ]
  particleSpawnScript [
   &InitializeParticle(
    Lifetime=2.5
    Mesh Scale=&Multiply_VectorByFloat(Vector=<1 1 1.2>, Float=1)
    Mesh Scale Mode=::NewEnumerator4
    UsePositionOffset=false
    Write Material Random=true
   )
   &CameraOffset(Camera Offset Amount=1)
  ]
  particleUpdateScript [
   &ParticleState()
   &ScaleColor(
    Scale Alpha=&Multiply_Float(A=0.7, B=1)
    Scale RGB=&Multiply_VectorByFloat(
     Vector=&MakeVectorFromLinearColorRGB(Color=&ColorFromCurve(ColorCurve=DataInterface() {propertyValues {
         AlphaCurve {keys []}
         BlueCurve {keys [{arriveTangent=-33.7362, interpMode=RCIM_Cubic, time=0.1676, value=1.8111}, <at 0.2769 0.5247>, <at 0.3205 1 RCIM_Cubic>]}
         GreenCurve {keys [{arriveTangent=-5.837, interpMode=RCIM_Cubic, time=0.1676, value=4.9}, <at 0.2769 2.339>, <at 0.3205 1 RCIM_Cubic>]}
         RedCurve {keys [{arriveTangent=3.7485, interpMode=RCIM_Cubic, time=0.1676, value=0}, <at 0.2769 3>, <at 0.3205 1 RCIM_Cubic>]}
        }}))
     Float=1
    )
    ScaleA=true
    ScaleRGB=true
   )
   &DynamicMaterialParameters(
    Index 0 Param 1=&FloatFromCurve(FloatCurve=DataInterface() {propertyValues {Curve {keys [
         {arriveTangent=0.6917, interpMode=RCIM_Cubic, leaveTangent=0.6917, tangentMode=RCTM_User, time=0, value=1}
         {arriveTangent=-4.5607, interpMode=RCIM_Cubic, leaveTangent=-4.5607, tangentMode=RCTM_User, time=0.4, value=0}
        ]}}})
    Module.Index0Param0Write=true
    Module.Index0Param1Write=true
    Module.Index0Param2Write=true
    Module.Index0Param3Write=true
   )
   &ScaleColor(
    Scale RGB=&Multiply_VectorByFloat(
     Vector=&MakeVector(
      X=&FloatFromCurve(FloatCurve=DataInterface() {propertyValues {Curve {keys [<at 0 1>,<at 0.4 1>,<at 0.5 1>,<at 0.7 1>]}}})
      Y=&FloatFromCurve(FloatCurve=DataInterface() {propertyValues {Curve {keys [<at 0 1>,<at 0.4 1>,<at 0.5 1>,<at 0.7 1>]}}})
      Z=&FloatFromCurve(FloatCurve=DataInterface() {propertyValues {Curve {keys [<at 0 1>,<at 0.4 1>,<at 0.5 1>,<at 0.7 1>]}}})
     )
     Float=1
    )
    ScaleA=true
    ScaleRGB=true
   ) {module=ScaleColor001}
  ]
  rendererClasses [&NiagaraMeshRendererProperties]
  renderers [{rendererClass=&NiagaraMeshRendererProperties, rendererIndex=0}]
 }
 Outline {
  emitterUpdateScript [
   &EmitterState(
    Inactive Response=::NewEnumerator1
    Life Cycle Mode=::NewEnumerator1
    Loop Behavior=::NewEnumerator1
    Loop Delay=0.9
    Loop Duration=3
    UseLoopDelay=false
   )
   &SpawnBurst_Instantaneous(Use Spawn Probability=false, UseLoopCountLimit=false)
  ]
  particleSpawnScript [
   &InitializeParticle(
    Color=<rgba 5 1.126 0.8772 1>
    Lifetime=3
    Mesh Scale=&Multiply_VectorByFloat(Vector=<1 1 1.2>, Float=1)
    Mesh Scale Mode=::NewEnumerator4
    UsePositionOffset=false
    Write Material Random=true
   )
   &CameraOffset(Camera Offset Amount=1)
  ]
  particleUpdateScript [
   &ParticleState()
   &ScaleColor(
    Scale Alpha=&Multiply_Float(
     A=0.5
     B=&FloatFromCurve(
      FloatCurve=DataInterface() {propertyValues {Curve {keys [
          {interpMode=RCIM_Cubic, tangentMode=RCTM_User, time=0, value=0}
          {arriveTangent=3.8173, interpMode=RCIM_Cubic, leaveTangent=3.8173, tangentMode=RCTM_User, time=0.3, value=1}
         ]}}}
      Scale Curve=0.2
     )
    )
    ScaleA=true
    ScaleRGB=true
   )
   &DynamicMaterialParameters(
    Index 0 Param 1=&FloatFromCurve(FloatCurve=DataInterface() {propertyValues {Curve {keys [<at 0 1 RCIM_Cubic>,<at 0.2 0 RCIM_Cubic>]}}})
    Module.Index0Param0Write=true
    Module.Index0Param1Write=true
    Module.Index0Param2Write=true
    Module.Index0Param3Write=true
   )
  ]
  rendererClasses [&NiagaraMeshRendererProperties]
  renderers [{rendererClass=&NiagaraMeshRendererProperties, rendererIndex=0}]
 }
}
summary {
 systemName=NS_Emote_WarmShade_Deploy
 emitters [
  {emitterName=MainMesh, rendererClasses [&NiagaraMeshRendererProperties]}
  {emitterName=BurstRay, rendererClasses [&NiagaraMeshRendererProperties]}
  {emitterName=Glitch_sprite, rendererClasses [&NiagaraMeshRendererProperties]}
  {emitterName=Glitch_sprite001, rendererClasses [&NiagaraMeshRendererProperties]}
  {emitterName=Outline, rendererClasses [&NiagaraMeshRendererProperties]}
  {emitterName=Glow, rendererClasses [&NiagaraSpriteRendererProperties]}
  {emitterName=Glitch, rendererClasses [&NiagaraMeshRendererProperties]}
 ]
 userVariables [
  User.bIsFrontEnd {type=NiagaraBool}
  User.bIsFrontEndPreview {type=NiagaraBool}
  User.RibbonIDOffset {type=NiagaraInt32}
 ]
}
