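# Looping, music-reactive Niagara effect for the WarmShade emote, built from a main mesh shape, side mesh, mirrored left/right rings and glow lights.
# A grid of square meshes and rising glow squares use a local SquareMovement module to shape equalizer-like bars.
# System MusicReactive and NPC audio amplitude drive ring, bar and side-mesh scale, color flashes and visibility.
# MainShape and OutLine emitters, plus their duplicates, are disabled.
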NiagaraSystem /BRCosmetics/Effects/Fort_Effects/Emotes/WarmShade/NS_Emote_WarmShade_Loop
dependencies {
 usedDataInterfaces [
  &NiagaraDataInterfacePlatformSet
  &NiagaraDataInterfaceCurve
  &NiagaraDataInterfaceColorCurve
  &NiagaraDataInterfaceArrayFloat3
  &NiagaraDataInterfaceArrayFloat
 ]
 usedDynamicInputs []
 usedModules [
  &SystemState
  &InterpolateOverTime
  &self:SystemSpawnScript.SystemScriptSource.SystemScriptGraph.NiagaraNodeAssignment_0.SetVariables_#1
  &self:SystemSpawnScript.SystemScriptSource.SystemScriptGraph.NiagaraNodeAssignment_1.SetVariables_#2
 ]
}
emitters {
 CenterRing {
  emitterUpdateScript [
   &EmitterState(
    Inactive Response=::NewEnumerator1
    Life Cycle Mode=::NewEnumerator1
    Loop Behavior=::NewEnumerator1
    Loop Duration Mode=::NewEnumerator1
    UseLoopDelay=false
   )
   &SpawnRate(SpawnRate=6, Use Spawn Probability=false)
  ]
  particleSpawnScript [
   &PartitionParticles()
   &InitializeParticle(
    Color=<rgba 10 0 0.9154 1>
    Lifetime Max=1.3
    Lifetime Min=0.8
    Lifetime Mode=::NewEnumerator1
    Mesh Scale Mode=::NewEnumerator3
    Mesh Uniform Scale=&RandomRangeFloat(Minimum=0.13, Maximum=0.18)
    Position Offset=&SelectVectorFromArray(
     Direct Array Index=$Particles.PartitionParticles.Partition
     Vector Selection Array=DataInterface() {propertyValues {FloatData [<-34 -1.5 7>, <34 -1.5 7>]}}
    )
    Sprite Size Mode=::NewEnumerator3
    Uniform Sprite Size=30
    UsePositionOffset=true
    Write Material Random=true
   )
   &Color(Color=&MultiplyLinearColor(Linear Color=$Particles.Color, Scale Factor=&MakeVector4(X=1, Y=1, Z=1, W=0.15)))
   /SetVariables(Particles.bool:NiagaraBool=&SetBoolByFloatComparison(A=&Subtract_Float(A=$Emitter.Age, B=$Emitter.float), B=-0.01)) {isSetParameters=true}
  ]
  particleUpdateScript [
   &ParticleState()
   &Timeline(Play Rate=0.02, Timeline End=100000)
   &CameraOffset()
   &DynamicMaterialParameters(
    Index 0 Param 1=&Multiply_Float(A=&MakeFloatFromVector(Vector=$Particles.Scale, Channel=::NewEnumerator0), B=1)
    Index 0 Param 2=50
    Index 0 Param 4=5
    Index 1 Param 1=15
    Index 1 Param 2=-0.07
    Index 1 Param 3=0.05
    Index 1 Param 4=$Particles.Timeline.Time
    Index 2 Param 1=3
    Module.Index0Param0Write=true
    Module.Index0Param1Write=true
    Module.Index0Param2Write=true
    Module.Index0Param3Write=true
    Module.Index1Param0Write=true
    Module.Index1Param1Write=true
    Module.Index1Param2Write=true
    Module.Index1Param3Write=true
    Module.Index2Param0Write=true
    Module.Index2Param1Write=true
    Module.Index2Param2Write=true
    Module.Index2Param3Write=true
    Module.Index3Param0Write=true
    Module.Index3Param1Write=true
    Module.Index3Param2Write=true
    Module.Index3Param3Write=true
    Write Parameter Index 1=true
    Write Parameter Index 2=true
    Write Parameter Index 3=true
   )
   &ScaleMeshSize(Scale Factor=&VectorFromFloat(Value=&FloatFromCurve(
      FloatCurve=DataInterface() {propertyValues {Curve {keys [{interpMode=RCIM_Cubic, tangentMode=RCTM_User, time=0, value=0.5}, {interpMode=RCIM_Cubic, tangentMode=RCTM_User, time=1, value=1}]}}}
      Scale Curve=0.4
     )))
   /SetVariables(Particles.MeshOrientation=&MakeQuatFromAxisAngle(Axis=<1 0 0>, AngleInDegrees=90)) {isSetParameters=true}
   &ScaleColor(
    Scale Alpha=&FloatFromCurve(FloatCurve=DataInterface() {propertyValues {Curve {keys [
         {interpMode=RCIM_Cubic, tangentMode=RCTM_User, time=0, value=0}
         {interpMode=RCIM_Cubic, tangentMode=RCTM_User, time=0.5, value=1}
         {tangentMode=RCTM_User, time=1, value=0}
        ]}}})
    Scale RGB=&Multiply_VectorByFloat(Vector=<1 1 1>, Float=&RandomRangeFloat(Minimum=1, Maximum=1))
    ScaleA=true
    ScaleRGB=true
   ) {module=ScaleColor001}
   &ScaleMeshSize(Scale Factor=&VectorFromFloat(Value=&Lerp_Float(
      A=&Waveform(Global Input Position=-0.17, "Waveform [1]"=::NewEnumerator0, "[1] Frequency"=1, "[1] Amplitude Min/Max"=<1 1.2>)
      B=&Waveform(Global Input Position=-0.17, "Waveform [1]"=::NewEnumerator0, "[1] Phase Scale"=3, "[1] Frequency"=1, "[1] Amplitude Min/Max"=<1.8 2.2>)
      Alpha=&FloatFromCurve(
       FloatCurve=DataInterface() {propertyValues {Curve {keys [{interpMode=RCIM_Cubic, leaveTangent=1.8912, tangentMode=RCTM_Break, time=0, value=0}, <at 1 1>]}}}
       CurveIndex=$System.MusicReactive
       Scale Curve=1.3
      )
     ))) {module=ScaleMeshSize001}
  ]
  rendererClasses [&NiagaraMeshRendererProperties]
  renderers [{rendererClass=&NiagaraMeshRendererProperties, rendererIndex=0}]
 }
 Glow {
  emitterUpdateScript [
   &EmitterState(Life Cycle Mode=::NewEnumerator1, Loop Behavior=::NewEnumerator1, Loop Duration Mode=::NewEnumerator1, UseLoopDelay=false)
   &SpawnBurst_Instantaneous(Spawn Count=2, Use Spawn Probability=false, UseLoopCountLimit=false)
  ]
  particleSpawnScript [
   &PartitionParticles()
   &InitializeParticle(
    Color=<rgba 0.3 0.006 0.0305 0.3>
    Position Offset=&SelectVectorFromArray(
     Direct Array Index=$Particles.PartitionParticles.Partition
     Vector Selection Array=DataInterface() {propertyValues {FloatData [<-40 -15 0>, <40 -15 0>]}}
    )
    Sprite Size Mode=::NewEnumerator3
    Uniform Sprite Size=350
    UsePositionOffset=true
    Write Material Random=true
   )
   &SpriteFacingAndAlignment(Sprite Alignment=<1 0 0>, Sprite Facing=<0 0 1>, WriteAlignment=true, WriteFacing=true)
   &DynamicMaterialParameters(
    Index 0 Param 1=3
    Index 0 Param 2=10
    Module.Index0Param0Write=true
    Module.Index0Param1Write=true
    Module.Index0Param2Write=true
    Module.Index0Param3Write=true
   )
  ]
  particleUpdateScript [
   &ParticleState(Kill Particles When Lifetime Has Elapsed=false)
   &SolveForcesAndVelocity(Clamp Velocity=false, Limit Acceleration=false)
   &FN_Light_Attributes(
    Color=<rgba 20 1.0131 2.0083 1>
    Create Color Attribute=true
    Create Exponent Attribute=true
    Create Radius Attribute=true
    Enable Particle Light Enable=true
    Exponent=3
    Radius=100
   )
  ]
  rendererClasses [&NiagaraSpriteRendererProperties, &NiagaraLightRendererProperties]
  renderers [
   {rendererClass=&NiagaraSpriteRendererProperties}
   {rendererClass=&NiagaraLightRendererProperties, rendererIndex=1}
  ]
 }
 MainMeshShape {
  emitterUpdateScript [
   &EmitterState(
    Inactive Response=::NewEnumerator1
    Life Cycle Mode=::NewEnumerator1
    Loop Behavior=::NewEnumerator1
    Loop Duration Mode=::NewEnumerator1
    UseLoopDelay=true
   )
   &SpawnBurst_Instantaneous(Use Spawn Probability=false, UseLoopCountLimit=false)
  ]
  particleSpawnScript [
   &InitializeParticle(
    Color=<rgba 0.5 0.5 0.5 1>
    Lifetime=2.5
    Mesh Scale=&Multiply_VectorByFloat(Vector=<1 1 1.2>, Float=1)
    Mesh Scale Mode=::NewEnumerator4
    UsePositionOffset=false
    Write Material Random=true
   )
   &CameraOffset(Camera Offset Amount=2)
  ]
  particleUpdateScript [
   &ParticleState(Kill Particles When Lifetime Has Elapsed=false)
   &ScaleColor(
    Scale Alpha=&Multiply_Float(A=1, B=1)
    Scale RGB=&Multiply_VectorByFloat(
     Vector=&MakeVectorFromLinearColorRGB(Color=&ColorFromCurve(ColorCurve=DataInterface() {propertyValues {
         AlphaCurve {keys []}
         BlueCurve {keys [<at 0.2 1 RCIM_Cubic>]}
         GreenCurve {keys [<at 0.2 1 RCIM_Cubic>]}
         RedCurve {keys [<at 0.2 1 RCIM_Cubic>]}
        }}))
     Float=1
    )
    ScaleA=true
    ScaleRGB=true
   )
   &DynamicMaterialParameters(Module.Index0Param0Write=true, Module.Index0Param1Write=true, Module.Index0Param2Write=true, Module.Index0Param3Write=true)
   &FN_Mesh_Attributes(
    Color=&MakeLinearColorFromVectorAndFloat(
     "Vector (RGB)"=&Multiply_Vector(A=&MakeVectorFromLinearColorRGB(Color=<rgba 5 1.126 0.8772 1>), B=<1 1 1>)
     "Float (Alpha)"=&Multiply_Float(A=0.2, B=1)
    )
    Create Color Attribute=true
    Write to Module Namespace=true
   )
   &FN_Mesh_Attributes(
    Color=&MakeLinearColorFromVectorAndFloat(
     "Vector (RGB)"=&Multiply_Vector(A=&MakeVectorFromLinearColorRGB(Color=<rgba 10 10 10 1>), B=<1 1 1>)
     "Float (Alpha)"=&Multiply_Float(A=0.2, B=1)
    )
    Create Color Attribute=true
    Write to Module Namespace=true
   ) {module=FN_Mesh_Attributes001}
  ]
  rendererClasses [&NiagaraMeshRendererProperties]
  renderers [
   {rendererClass=&NiagaraMeshRendererProperties, rendererIndex=0}
   {rendererClass=&NiagaraMeshRendererProperties, rendererIndex=1}
   {rendererClass=&NiagaraMeshRendererProperties, rendererIndex=2}
  ]
 }
 MainShape {
  bEnabled=false
  emitterUpdateScript [
   &EmitterState(
    Inactive Response=::NewEnumerator1
    Life Cycle Mode=::NewEnumerator1
    Loop Behavior=::NewEnumerator1
    Loop Duration Mode=::NewEnumerator1
    UseLoopDelay=true
   )
   &SpawnBurst_Instantaneous(Use Spawn Probability=false, UseLoopCountLimit=false)
  ]
  particleSpawnScript [
   &InitializeParticle(
    Color=<rgba 0.5 0.5 0.5 1>
    Lifetime=2.5
    Mesh Scale=&Multiply_VectorByFloat(Vector=<1 1 1.2>, Float=1)
    Mesh Scale Mode=::NewEnumerator4
    UsePositionOffset=false
    Write Material Random=true
   )
   &CameraOffset(Camera Offset Amount=1)
  ]
  particleUpdateScript [
   &ParticleState(Kill Particles When Lifetime Has Elapsed=false)
   &ScaleColor(
    Scale Alpha=&Multiply_Float(A=0.7, B=1)
    Scale RGB=&Multiply_VectorByFloat(
     Vector=&MakeVectorFromLinearColorRGB(Color=&ColorFromCurve(ColorCurve=DataInterface() {propertyValues {
         AlphaCurve {keys []}
         BlueCurve {keys [<at 0.2 1 RCIM_Cubic>]}
         GreenCurve {keys [<at 0.2 1 RCIM_Cubic>]}
         RedCurve {keys [<at 0.2 1 RCIM_Cubic>]}
        }}))
     Float=1
    )
    ScaleA=true
    ScaleRGB=true
   )
   &DynamicMaterialParameters(Module.Index0Param0Write=true, Module.Index0Param1Write=true, Module.Index0Param2Write=true, Module.Index0Param3Write=true)
  ]
  rendererClasses [&NiagaraMeshRendererProperties]
  renderers [{rendererClass=&NiagaraMeshRendererProperties, rendererIndex=0}]
 }
 MainShape001 {
  bEnabled=false
  emitterUpdateScript [
   &EmitterState(
    Inactive Response=::NewEnumerator1
    Life Cycle Mode=::NewEnumerator1
    Loop Behavior=::NewEnumerator1
    Loop Duration Mode=::NewEnumerator1
    UseLoopDelay=true
   )
   &SpawnBurst_Instantaneous(Use Spawn Probability=false, UseLoopCountLimit=false)
  ]
  particleSpawnScript [
   &InitializeParticle(
    Color=<rgba 0.5 0.5 0.5 1>
    Lifetime=2.5
    Mesh Scale=&Multiply_VectorByFloat(Vector=<1 1 1.2>, Float=1)
    Mesh Scale Mode=::NewEnumerator4
    UsePositionOffset=false
    Write Material Random=true
   )
   &CameraOffset(Camera Offset Amount=1)
  ]
  particleUpdateScript [
   &ParticleState(Kill Particles When Lifetime Has Elapsed=false)
   &ScaleColor(
    Scale Alpha=&Multiply_Float(A=0.7, B=1)
    Scale RGB=&Multiply_VectorByFloat(
     Vector=&MakeVectorFromLinearColorRGB(Color=&ColorFromCurve(ColorCurve=DataInterface() {propertyValues {
         AlphaCurve {keys []}
         BlueCurve {keys [<at 0.2 1 RCIM_Cubic>]}
         GreenCurve {keys [<at 0.2 1 RCIM_Cubic>]}
         RedCurve {keys [<at 0.2 1 RCIM_Cubic>]}
        }}))
     Float=1
    )
    ScaleA=true
    ScaleRGB=true
   )
   &DynamicMaterialParameters(Module.Index0Param0Write=true, Module.Index0Param1Write=true, Module.Index0Param2Write=true, Module.Index0Param3Write=true)
  ]
  rendererClasses [&NiagaraMeshRendererProperties]
  renderers [{rendererClass=&NiagaraMeshRendererProperties, rendererIndex=0}]
 }
 MeshRing {
  emitterUpdateScript [
   &EmitterState(
    Inactive Response=::NewEnumerator1
    Life Cycle Mode=::NewEnumerator1
    Loop Behavior=::NewEnumerator1
    Loop Duration Mode=::NewEnumerator1
    UseLoopDelay=false
   )
   &SpawnBurst_Instantaneous(Spawn Count=2, Use Spawn Probability=false, UseLoopCountLimit=false)
  ]
  particleSpawnScript [
   &PartitionParticles()
   &InitializeParticle(
    Color Mode=::NewEnumerator0
    Mesh Scale Mode=::NewEnumerator3
    Position Offset=&SelectVectorFromArray(
     Direct Array Index=$Particles.PartitionParticles.Partition
     Vector Selection Array=DataInterface() {propertyValues {FloatData [<34 -1.5 10>, <-34 -1.5 10>]}}
    )
    UsePositionOffset=true
    Write Material Random=true
   )
  ]
  particleUpdateScript [
   &ParticleState(Kill Particles When Lifetime Has Elapsed=false)
   &ScaleMeshSize(Scale Factor=&Multiply_VectorByFloat(
     Vector=<1 1 1>
     Float=&Lerp_Float(A=0.7, B=1.2, Alpha=&Power(a=$System.MusicReactive, b=2))
    ))
   &UpdateMeshOrientation(
    Rotation Rate=&SelectFloatFromArray(Direct Array Index=$Particles.PartitionParticles.Partition, Float Selection Array=DataInterface() {propertyValues {FloatData [-0.5,0.5]}})
    Rotation Vector=<0 0 1>
   )
   &ScaleColor(
    Scale Alpha=0.8
    Scale RGB=&Multiply_VectorByFloat(
     Vector=&Multiply_Vector(
      A=<1 1 1>
      B=&Lerp_Vector(
       A=<1 1 1>
       B=<50 1 1>
       Alpha=&Power(a=&Multiply_Float(A=$System.MusicReactive, B=$NPC.FortniteNPC.EventAudio_Amplitude), b=15)
      )
     )
     Float=2
    )
    ScaleA=true
    ScaleRGB=true
   )
   &FN_Mesh_Attributes(
    Color=&MakeLinearColorFromVectorAndFloat(
     "Vector (RGB)"=&Multiply_Vector(
      A=&MakeVectorFromLinearColorRGB(Color=&MultiplyLinearColor(Linear Color=<rgba 0.099 1 0.8976 1>, Scale Factor=&MakeVector4_FromFloat(Float=2.5)))
      B=&Lerp_Vector(
       A=<1 1 1>
       B=<40 1 1>
       Alpha=&Power(a=&Multiply_Float(A=$System.MusicReactive, B=$NPC.FortniteNPC.EventAudio_Amplitude), b=20)
      )
     )
     "Float (Alpha)"=0.75
    )
    Create Color Attribute=true
    Create Scale Attribute=true
    Scale=&Multiply_VectorByFloat(
     Vector=<1 1 1>
     Float=&Multiply_Float(A=0.8, B=&Lerp_Float(A=0.5, B=1, Alpha=&Power(a=$System.MusicReactive, b=2)))
    )
    Write to Module Namespace=true
   )
   &FN_Mesh_Attributes(
    Color=&MakeLinearColorFromVectorAndFloat(
     "Vector (RGB)"=&Multiply_Vector(
      A=&MakeVectorFromLinearColorRGB(Color=&MultiplyLinearColor(Linear Color=<rgba 1 0.2285 0.2584 1>, Scale Factor=&MakeVector4_FromFloat(Float=5)))
      B=&Lerp_Vector(
       A=<0.5 0.5 0.5>
       B=<5 5 1>
       Alpha=&Power(a=&Multiply_Float(A=$System.MusicReactive, B=$NPC.FortniteNPC.EventAudio_Amplitude), b=10)
      )
     )
     "Float (Alpha)"=0.6
    )
    Create Color Attribute=true
    Create Scale Attribute=true
    Scale=&Multiply_VectorByFloat(
     Vector=<1 1 2>
     Float=&Multiply_Float(A=1.25, B=&Lerp_Float(A=0.7, B=1.1, Alpha=&Power(a=$System.MusicReactive, b=2)))
    )
    Write to Module Namespace=true
   ) {module=FN_Mesh_Attributes001}
   /SetVariables(Particles.Position002:NiagaraPosition=&AddVectorToPosition(Position=$Particles.Position, Vector=<0 0 -3>)) {isSetParameters=true}
   /SetVariables(Particles.MeshOrientation02:Quat4f=/Niagara/DynamicInputs/Rotation/InvertQuaternion(Quat:Quat4f=$Particles.MeshOrientation)) {isSetParameters=true}
  ]
  rendererClasses [&NiagaraMeshRendererProperties]
  renderers [
   {rendererClass=&NiagaraMeshRendererProperties, rendererIndex=0}
   {rendererClass=&NiagaraMeshRendererProperties, rendererIndex=1}
   {rendererClass=&NiagaraMeshRendererProperties, rendererIndex=2}
  ]
 }
 OutLine {
  bEnabled=false
  emitterUpdateScript [
   &EmitterState(
    Inactive Response=::NewEnumerator1
    Life Cycle Mode=::NewEnumerator1
    Loop Behavior=::NewEnumerator1
    Loop Delay=0.9
    Loop Duration Mode=::NewEnumerator1
    UseLoopDelay=false
   )
   &SpawnBurst_Instantaneous(Use Spawn Probability=false, UseLoopCountLimit=false)
  ]
  particleSpawnScript [
   &InitializeParticle(
    Color=<rgba 10 10 10 1>
    Lifetime=3
    Mesh Scale=&Multiply_VectorByFloat(Vector=<1 1 1.2>, Float=1)
    Mesh Scale Mode=::NewEnumerator4
    UsePositionOffset=false
    Write Material Random=true
   )
   &CameraOffset(Camera Offset Amount=1)
  ]
  particleUpdateScript [
   &ParticleState(Kill Particles When Lifetime Has Elapsed=false)
   &ScaleColor(Scale Alpha=&Multiply_Float(A=0.2, B=1), ScaleA=true, ScaleRGB=true)
   &DynamicMaterialParameters(Module.Index0Param0Write=true, Module.Index0Param1Write=true, Module.Index0Param2Write=true, Module.Index0Param3Write=true)
  ]
  rendererClasses [&NiagaraMeshRendererProperties]
  renderers [{rendererClass=&NiagaraMeshRendererProperties, rendererIndex=0}]
 }
 OutLine001 {
  bEnabled=false
  emitterUpdateScript [
   &EmitterState(
    Inactive Response=::NewEnumerator1
    Life Cycle Mode=::NewEnumerator1
    Loop Behavior=::NewEnumerator1
    Loop Delay=0.9
    Loop Duration Mode=::NewEnumerator1
    UseLoopDelay=false
   )
   &SpawnBurst_Instantaneous(Use Spawn Probability=false, UseLoopCountLimit=false)
  ]
  particleSpawnScript [
   &InitializeParticle(
    Color=<rgba 5 1.126 0.8772 1>
    Lifetime=3
    Mesh Scale=&Multiply_VectorByFloat(Vector=<1 1 1.2>, Float=1)
    Mesh Scale Mode=::NewEnumerator4
    UsePositionOffset=false
    Write Material Random=true
   )
   &CameraOffset(Camera Offset Amount=1)
  ]
  particleUpdateScript [
   &ParticleState(Kill Particles When Lifetime Has Elapsed=false)
   &ScaleColor(Scale Alpha=&Multiply_Float(A=0.3, B=1), ScaleA=true, ScaleRGB=true)
   &DynamicMaterialParameters(Module.Index0Param0Write=true, Module.Index0Param1Write=true, Module.Index0Param2Write=true, Module.Index0Param3Write=true)
  ]
  rendererClasses [&NiagaraMeshRendererProperties]
  renderers [{rendererClass=&NiagaraMeshRendererProperties, rendererIndex=0}]
 }
 SideMesh {
  emitterUpdateScript [
   &EmitterState(
    Inactive Response=::NewEnumerator1
    Life Cycle Mode=::NewEnumerator1
    Loop Behavior=::NewEnumerator1
    Loop Duration Mode=::NewEnumerator1
    UseLoopDelay=false
   )
   &SpawnBurst_Instantaneous(Use Spawn Probability=false, UseLoopCountLimit=false)
  ]
  particleSpawnScript [
   &InitializeParticle(
    Color=<rgba 2 2 2 0.7>
    Mesh Scale=<1 1 1.2>
    Mesh Scale Mode=::NewEnumerator4
    Position Offset=<-0.5 -3 -1.5>
    UsePositionOffset=true
    Write Material Random=true
   )
   &CameraOffset(Camera Offset Amount=1)
  ]
  particleUpdateScript [
   &ParticleState(Kill Particles When Lifetime Has Elapsed=false)
   &ScaleMeshSize(Scale Factor=&MakeVector(X=&Lerp_Float(A=0.85, B=1.08, Alpha=&Power(a=$System.MusicReactive, b=2)), Y=1, Z=1))
   &ScaleColor(
    Scale Alpha=0.4
    Scale RGB=&Multiply_Vector(
     A=<1 1 1>
     B=&Lerp_Vector(
      A=<1 1 1>
      B=<30 1 1>
      Alpha=&Power(a=&Multiply_Float(A=$System.MusicReactive, B=$NPC.FortniteNPC.EventAudio_Amplitude), b=20)
     )
    )
    ScaleA=true
    ScaleRGB=true
   )
  ]
  rendererClasses [&NiagaraMeshRendererProperties]
  renderers [{rendererClass=&NiagaraMeshRendererProperties, rendererIndex=0}]
 }
 SquareFront {
  emitterUpdateScript [
   &EmitterState(
    Inactive Response=::NewEnumerator1
    Life Cycle Mode=::NewEnumerator1
    Loop Behavior=::NewEnumerator1
    Loop Duration Mode=::NewEnumerator1
    UseLoopDelay=false
   )
   /Niagara/Modules/Spawn/Location/SpawnParticlesInGrid(X Count=30, Y Count=2, Z Count=0)
   /SetVariables(Emitter.Scanlines_Dark:NiagaraFloat=&FN_PickFloatDependOnPlatform(TargetPlatform=DataInterface() {propertyValues {Platforms {qualityLevelMask=-2}}}, TargetPlatformActive=0.5, TargetPlatformNoActive=1)) {isSetParameters=true}
  ]
  particleSpawnScript [
   /SetVariables(Particles.Randomfloat:NiagaraFloat=&RandomRangeFloat(Minimum=0, Maximum=1)) {isSetParameters=true}
   &InitializeParticle(
    Lifetime=&RandomRangeFloat(Minimum=0.3, Maximum=1)
    Mesh Scale Mode=::NewEnumerator1
    Mesh Uniform Scale Max=0.6
    Mesh Uniform Scale Min=0.3
    Sprite Size Mode=::NewEnumerator1
    Uniform Sprite Size Min=10
    UsePositionOffset=false
    Write Material Random=true
   )
   &SystemLocation(MaskSpawnGroup=false, Offset=<10 -35 5>)
   /Niagara/Modules/Spawn/Location/V2/GridLocation(
    Module.Jitter Inside Grid Cells=false
    Module.Override Random Seed=false
    Randomize Placement Within Cell:Vector3f=<0 0 0>
    XYZ Dimensions:Vector3f=<5 6 0>
   )
  ]
  particleUpdateScript [
   &ParticleState(Kill Particles When Lifetime Has Elapsed=false, Loop Particles Lifetime=true)
   &self:SquareMovement(Disapear=0.1, Fade=170, Offset=85) {module=SquareMovement}
   &ScaleMeshSize(Scale Factor=&Multiply_Vector(
     A=&MakeVector(
      X=1
      Y=1
      Z=&Multiply_Float(A=&Lerp_Float(A=0, B=10, Alpha=$Particles.MypositionSet), B=3)
     )
     B=&MakeVector(
      X=1
      Y=1
      Z=&Multiply_Float(
       A=&FloatFromCurve(
        FloatCurve=DataInterface() {propertyValues {Curve {keys [
            {interpMode=RCIM_Cubic, tangentMode=RCTM_User, time=0, value=0.4}
            {interpMode=RCIM_Cubic, tangentMode=RCTM_User, time=0.5, value=1.2}
            {tangentMode=RCTM_User, time=1, value=0.4}
           ]}}}
        Scale Curve=0.7
       )
       B=&Lerp_Float(
        A=&Multiply_Float(A=$NPC.FortniteNPC.EventAudio_Amplitude, B=$NPC.FortniteNPC.EventAudio_Amplitude)
        B=1
        Alpha=$System.MusicReactivePreview
       )
      )
     )
    ))
   &Color(Color=&LerpLinearColors(StartColor=<rgba 2 0.8602 1.2041 1>, EndColor=<rgba 2 0.0589 0.2207 1>, LerpFactor=$Particles.Randomfloat))
   &ScaleColor(
    Color Value To Scale=$Particles.Color
    Scale Alpha=&Multiply_Float(
     A=$Particles.CustomDisapear
     B=&FloatFromCurve(
      FloatCurve=DataInterface() {propertyValues {Curve {keys [
          {interpMode=RCIM_Cubic, tangentMode=RCTM_User, time=0, value=0}
          {interpMode=RCIM_Cubic, tangentMode=RCTM_User, time=0.5, value=1}
          {tangentMode=RCTM_User, time=1, value=0}
         ]}}}
      Scale Curve=&MakeCustomFloatFromBool(Bool=&SetBoolByFloatComparison(A=$NPC.FortniteNPC.EventAudio_Amplitude, B=0), True Float=1, False Float=0)
     )
    )
    Scale RGB=&Multiply_VectorByFloat(
     Vector=&MakeVectorFromLinearColorRGB(Color=&ColorFromCurve(ColorCurve=DataInterface() {propertyValues {
         AlphaCurve {keys []}
         BlueCurve {keys [
           <at 0 1 RCIM_Cubic>
           <at 0.1479 2.4811>
           {arriveTangent=-2.7405, interpMode=RCIM_Cubic, leaveTangent=-2.7405, time=0.3483, value=1.3807}
           <at 0.51 1.4887>
          ]}
         GreenCurve {keys [
           <at 0 1 RCIM_Cubic>
           <at 0.1479 1.1596>
           {arriveTangent=-1.2808, interpMode=RCIM_Cubic, leaveTangent=-1.2808, time=0.3483, value=0.0708}
           <at 0.51 0.6957>
          ]}
         RedCurve {keys [<at 0 1 RCIM_Cubic>, <at 0.1479 5>, {arriveTangent=-5.5227, interpMode=RCIM_Cubic, leaveTangent=-5.5227, time=0.3483, value=5}, <at 0.51 3>]}
        }}))
     Float=&RandomRangeFloat(Minimum=0.5, Maximum=1.2)
    )
    ScaleA=true
    ScaleRGB=true
   )
   &DynamicMaterialParameters(
    Index 0 Param 1=&MakeFloatFromVector(Vector=&Multiply_Vector(A=<1 1 1>, B=$Particles.Scale), Channel=::NewEnumerator0)
    Index 0 Param 2=&MakeFloatFromVector(Vector=&Multiply_Vector(A=<1 1 1>, B=$Particles.Scale), Channel=::NewEnumerator1)
    Index 0 Param 3=&MakeFloatFromVector(Vector=&Multiply_Vector(A=<1 1 1>, B=$Particles.Scale), Channel=::NewEnumerator2)
    Index 1 Param 1=&Lerp_Float(A=-0.5, B=1, Alpha=$Particles.Randomfloat)
    Module.Index0Param0Write=true
    Module.Index0Param1Write=true
    Module.Index0Param2Write=true
    Module.Index0Param3Write=true
    Module.Index1Param0Write=true
    Module.Index1Param1Write=true
    Module.Index1Param2Write=true
    Module.Index1Param3Write=true
    Write Parameter Index 1=true
   )
   &FN_Mesh_Attributes(
    Color=&MakeLinearColorFromVectorAndFloat(
     "Vector (RGB)"=&Multiply_VectorByFloat(
      Vector=&MakeVectorFromLinearColorRGB(Color=&ColorFromCurve(ColorCurve=DataInterface() {propertyValues {
          AlphaCurve {keys []}
          BlueCurve {keys [<at 0 1 RCIM_Cubic>,<at 0.1893 1>,<at 0.3892 0.2>,<at 0.5866 0.326>,<at 1 0 RCIM_Cubic>]}
          GreenCurve {keys [<at 0 0 RCIM_Cubic>,<at 0.1893 0.4287>,<at 0.3892 0.1621>,<at 0.5866 1>,<at 1 0.3153 RCIM_Cubic>]}
          RedCurve {keys [<at 0 0.0057 RCIM_Cubic>,<at 0.1893 0>,<at 0.3892 0>,<at 0.5866 0>,<at 1 0.5 RCIM_Cubic>]}
         }}))
      Float=&RandomRangeFloat(Minimum=0.2, Maximum=1.5)
     )
     "Float (Alpha)"=&Multiply_Float(
      A=&Multiply_Float(
       A=$Particles.CustomDisapear
       B=&FloatFromCurve(
        FloatCurve=DataInterface() {propertyValues {Curve {keys [
            {interpMode=RCIM_Cubic, tangentMode=RCTM_User, time=0, value=0}
            {interpMode=RCIM_Cubic, tangentMode=RCTM_User, time=0.5, value=1}
            {tangentMode=RCTM_User, time=1, value=0}
           ]}}}
        Scale Curve=0.6
       )
      )
      B=1
     )
    )
    Create Color Attribute=true
    Write to Module Namespace=true
   )
   &FN_Mesh_Attributes(
    Color=&MakeLinearColorFromVectorAndFloat(
     "Vector (RGB)"=&Multiply_Vector(
      A=&Multiply_VectorByFloat(
       Vector=&MakeVectorFromLinearColorRGB(Color=&ColorFromCurve(ColorCurve=DataInterface() {propertyValues {
           AlphaCurve {keys []}
           BlueCurve {keys [
             <at 0 1 RCIM_Cubic>
             <at 0.1479 2.4811>
             {arriveTangent=-6.0719, interpMode=RCIM_Cubic, leaveTangent=-6.0719, time=0.3483, value=1.6544}
             <at 0.51 0.2823>
            ]}
           GreenCurve {keys [
             <at 0 1 RCIM_Cubic>
             <at 0.1479 1.1596>
             {arriveTangent=-0.0386, interpMode=RCIM_Cubic, leaveTangent=-0.0386, time=0.3483, value=5}
             <at 0.51 1.1456>
            ]}
           RedCurve {keys [<at 0 1 RCIM_Cubic>, <at 0.1479 5>, {arriveTangent=-5.5227, interpMode=RCIM_Cubic, leaveTangent=-5.5227, time=0.3483, value=0}, <at 0.51 3>]}
          }}))
       Float=&RandomRangeFloat(Minimum=0.1, Maximum=0.4)
      )
      B=&MakeVectorFromLinearColorRGB(Color=<rgba 1.295 0.7124 3 1>)
     )
     "Float (Alpha)"=&Multiply_Float(
      A=$Particles.CustomDisapear
      B=&FloatFromCurve(
       FloatCurve=DataInterface() {propertyValues {Curve {keys [
           {interpMode=RCIM_Cubic, tangentMode=RCTM_User, time=0, value=0}
           {interpMode=RCIM_Cubic, tangentMode=RCTM_User, time=0.5, value=1}
           {tangentMode=RCTM_User, time=1, value=0}
          ]}}}
       Scale Curve=0.3
      )
     )
    )
    Create Color Attribute=true
    Create Scale Attribute=true
    Scale=&Multiply_VectorByFloat(
     Vector=&Multiply_Vector(
      A=&MakeVector(
       X=1.2
       Y=1.2
       Z=&Multiply_Float(
        A=&Lerp_Float(A=0.2, B=&RandomRangeFloat(Minimum=10, Maximum=10), Alpha=$Particles.MypositionSet)
        B=4
       )
      )
      B=&MakeVector(
       X=1
       Y=1
       Z=&FloatFromCurve(
        FloatCurve=DataInterface() {propertyValues {Curve {keys [
            {interpMode=RCIM_Cubic, tangentMode=RCTM_User, time=0, value=0.6}
            {interpMode=RCIM_Cubic, tangentMode=RCTM_User, time=0.5, value=1.2}
            {tangentMode=RCTM_User, time=1, value=0.6}
           ]}}}
        Scale Curve=&Multiply_Float(A=1, B=&Multiply_Float(A=$NPC.FortniteNPC.EventAudio_Amplitude, B=$NPC.FortniteNPC.EventAudio_Amplitude))
       )
      )
     )
     Float=&RandomRangeFloat(Minimum=0.3, Maximum=0.5)
    )
    Write to Module Namespace=true
   ) {module=FN_Mesh_Attributes001}
  ]
  rendererClasses [&NiagaraMeshRendererProperties]
  renderers [
   {rendererClass=&NiagaraMeshRendererProperties, rendererIndex=0}
   {rendererClass=&NiagaraMeshRendererProperties, rendererIndex=1}
   {rendererClass=&NiagaraMeshRendererProperties, rendererIndex=2}
  ]
 }
 SquareGlow {
  emitterUpdateScript [
   &EmitterState(Inactive Response=::NewEnumerator1, Life Cycle Mode=::NewEnumerator1, Loop Duration=1, UseLoopDelay=false)
   &SpawnRate(SpawnRate=20, Use Spawn Probability=false)
  ]
  particleSpawnScript [
   &InitializeParticle(
    Color=<rgba 1.5 1.5 1.5 3>
    Mesh Scale=&Multiply_VectorByFloat(
     Vector=&RandomRangeVector(Minimum=<2 2 0.6>, Maximum=<3 3 0.8>)
     Float=&RandomRangeFloat(Minimum=1, Maximum=1)
    )
    Mesh Scale Mode=::NewEnumerator4
    Position Offset=<10 0 30>
    Sprite Size Mode=::NewEnumerator3
    UsePositionOffset=false
    Write Material Random=true
   )
   &SystemLocation(MaskSpawnGroup=false, Offset=&RandomRangeVector(Minimum=<55 -40 10>, Maximum=<-55 -40 10>))
   /SetVariables(Particles.Random=&RandomRangeFloat(Minimum=0, Maximum=1)) {isSetParameters=true}
   &InitialMeshOrientation(Mesh Orientation Mode=::NewEnumerator2, Rotation=<0 0 0.1>, Use Rotation Vector=true)
   &CameraOffset()
  ]
  particleUpdateScript [
   &ParticleState()
   &self:SquareMovement(Disapear=0.1, Fade=170, Offset=85) {module=SquareMovement}
   &SolveForcesAndVelocity(Clamp Velocity=false, Limit Acceleration=false)
   &ScaleColor(
    Scale Alpha=4
    Scale RGB=&Multiply_VectorByFloat(
     Vector=&MakeVectorFromLinearColorRGB(Color=&ColorFromCurve(
       ColorCurve=DataInterface() {propertyValues {
         AlphaCurve {keys [<at 0 1 RCIM_Cubic>]}
         BlueCurve {keys [<at 0.0176 0.8136>,<at 0.1843 0.0889>,<at 0.4216 1>,<at 0.548 0.1268>,<at 0.7273 0.2395>,<at 1 0.0671 RCIM_Cubic>]}
         GreenCurve {keys [<at 0.0176 0.0856>,<at 0.1843 0.0889>,<at 0.4216 0.5705>,<at 0.548 0.0371>,<at 0.7273 0.2395>,<at 1 0.4357 RCIM_Cubic>]}
         RedCurve {keys [<at 0.0176 1>,<at 0.1843 1>,<at 0.4216 0>,<at 0.548 1>,<at 0.7273 1>,<at 1 1 RCIM_Cubic>]}
        }}
       CurveIndex=&Multiply_Float(A=$Particles.NormalizedAge, B=$Particles.Random)
      ))
     Float=0.25
    )
    ScaleA=true
    ScaleRGB=true
   )
   &ScaleMeshSize(Scale Factor=&MakeVector(
     X=&FN_FrontEndFloatModifier(Gameplay=1, Front End=0.7)
     Y=&FN_FrontEndFloatModifier(Gameplay=1, Front End=0.7)
     Z=&FloatFromCurve(
      FloatCurve=DataInterface() {propertyValues {Curve {keys [
          {interpMode=RCIM_Cubic, tangentMode=RCTM_User, time=0, value=0}
          {interpMode=RCIM_Cubic, tangentMode=RCTM_User, time=0.5, value=27}
          {tangentMode=RCTM_User, time=1, value=0}
         ]}}}
      Scale Curve=$Particles.MypositionSet
     )
    ))
  ]
  rendererClasses [&NiagaraMeshRendererProperties]
  renderers [{rendererClass=&NiagaraMeshRendererProperties, rendererIndex=0}]
 }
 SquareRingL {
  emitterUpdateScript [
   &EmitterState(
    Inactive Response=::NewEnumerator1
    Life Cycle Mode=::NewEnumerator1
    Loop Behavior=::NewEnumerator1
    Loop Duration Mode=::NewEnumerator1
    UseLoopDelay=false
   )
   &SpawnBurst_Instantaneous(Spawn Count=50, Use Spawn Probability=false, UseLoopCountLimit=false)
  ]
  particleSpawnScript [
   &InitializeParticle(
    Color=<rgba 1.5 1.5 1.5 1>
    Mesh Scale Mode=::NewEnumerator1
    Mesh Uniform Scale Max=7
    Mesh Uniform Scale Min=3
    Position Offset=<-34 -1.5 7>
    Sprite Size Mode=::NewEnumerator3
    UsePositionOffset=true
    Write Material Random=true
   )
   &ShapeLocation(
    MaskSpawnGroup=false
    Override Seed=false
    Ring / Disc Distribution Mode:ENiagaraSphereDistributionMode=::NewEnumerator1
    Ring Radius=22
    Shape Primitive=::NewEnumerator4
    U Position=&ReturnNormalizedExecIndex()
    UseOwnerScale=true
   )
   &InitialMeshOrientation(Mesh Orientation Mode=::NewEnumerator2, Orientation Vector=$Particles.ShapeLocation.ShapeVector, Use Rotation Vector=false)
  ]
  particleUpdateScript [
   &ParticleState(Kill Particles When Lifetime Has Elapsed=false)
   &ScaleMeshSize(Scale Factor=&Multiply_VectorByFloat(
     Vector=&Multiply_Vector(
      A=&MakeVector(
       X=&Multiply_Float(
        A=&Multiply_Float(
         A=&UniformAOrBFloat(A=1, B=2.5, Evaluation Type=::NewEnumerator0, Recalculate Random Each Loop=true, Override Seed=false)
         B=&Sine(Normalized Angle=$Emitter.Age, Period=&RandomRangeFloat(Minimum=0.2, Maximum=1), Scale=2, Bias=1.5)
        )
        B=&Lerp_Float(
         A=&Power(a=&Multiply_Float(A=$System.MusicReactive, B=$NPC.FortniteNPC.EventAudio_Amplitude), b=1)
         B=1
         Alpha=$System.MusicReactivePreview
        )
       )
       Y=1
       Z=1
      )
      B=<1 1.5 1.5>
     )
     Float=0.025
    ))
   &ScaleColor(
    Scale Alpha=0.7
    Scale RGB=&Multiply_VectorByFloat(
     Vector=&MakeVectorFromLinearColorRGB(Color=&ColorFromCurve(
       ColorCurve=DataInterface() {propertyValues {
         AlphaCurve {keys []}
         BlueCurve {keys [<at 0 0.1859 RCIM_Cubic>,<at 0.5378 0.9502>,<at 0.7913 0>,<at 1 0.0208 RCIM_Cubic>]}
         GreenCurve {keys [<at 0 0.0762 RCIM_Cubic>,<at 0.5378 1>,<at 0.7913 0>,<at 1 0.0208 RCIM_Cubic>]}
         RedCurve {keys [<at 0 1 RCIM_Cubic>,<at 0.5378 0.009>,<at 0.7913 1>,<at 1 1 RCIM_Cubic>]}
        }}
       CurveIndex=&RandomRangeFloat(Minimum=0, Maximum=1)
      ))
     Float=0.5
    )
    ScaleA=true
    ScaleRGB=true
   ) {module=ScaleColor001}
   &ScaleColor(
    Scale Alpha=&Multiply_Float(
     A=&RandomRangeFloat(Minimum=0.5, Maximum=1)
     B=&MakeCustomFloatFromBool(Bool=&SetBoolByFloatComparison(A=$NPC.FortniteNPC.EventAudio_Amplitude, B=0), True Float=1, False Float=0)
    )
    Scale RGB=&Multiply_VectorByFloat(
     Vector=&MakeVectorFromLinearColorRGB(Color=&ColorFromCurve(
       ColorCurve=DataInterface() {propertyValues {
         AlphaCurve {keys []}
         BlueCurve {keys [<at 0 0.1859 RCIM_Cubic>,<at 0.5378 0.5832>,<at 0.7913 0>,<at 1 0.0208 RCIM_Cubic>]}
         GreenCurve {keys [<at 0 0.0762 RCIM_Cubic>,<at 0.5378 0.987>,<at 0.7913 0>,<at 1 0.0208 RCIM_Cubic>]}
         RedCurve {keys [<at 0 1 RCIM_Cubic>,<at 0.5378 2>,<at 0.7913 1>,<at 1 1 RCIM_Cubic>]}
        }}
       CurveIndex=&RandomRangeFloat(Minimum=0, Maximum=1)
      ))
     Float=0.5
    )
    ScaleA=true
    ScaleRGB=true
   )
  ]
  rendererClasses [&NiagaraMeshRendererProperties]
  renderers [{rendererClass=&NiagaraMeshRendererProperties, rendererIndex=0}]
 }
 SquareRingR {
  emitterUpdateScript [
   &EmitterState(
    Inactive Response=::NewEnumerator1
    Life Cycle Mode=::NewEnumerator1
    Loop Behavior=::NewEnumerator1
    Loop Duration Mode=::NewEnumerator1
    UseLoopDelay=false
   )
   &SpawnBurst_Instantaneous(Spawn Count=50, Use Spawn Probability=false, UseLoopCountLimit=false)
  ]
  particleSpawnScript [
   &InitializeParticle(
    Color=<rgba 1.5 1.5 1.5 1>
    Mesh Scale Mode=::NewEnumerator1
    Mesh Uniform Scale Max=7
    Mesh Uniform Scale Min=3
    Position Offset=<34 -1.5 7>
    Sprite Size Mode=::NewEnumerator3
    UsePositionOffset=true
    Write Material Random=true
   )
   &ShapeLocation(
    MaskSpawnGroup=false
    Override Seed=false
    Ring / Disc Distribution Mode:ENiagaraSphereDistributionMode=::NewEnumerator1
    Ring Radius=22
    Shape Primitive=::NewEnumerator4
    U Position=&ReturnNormalizedExecIndex()
    UseOwnerScale=true
   )
   &InitialMeshOrientation(Mesh Orientation Mode=::NewEnumerator2, Orientation Vector=$Particles.ShapeLocation.ShapeVector, Use Rotation Vector=false)
  ]
  particleUpdateScript [
   &ParticleState(Kill Particles When Lifetime Has Elapsed=false)
   &ScaleMeshSize(Scale Factor=&Multiply_VectorByFloat(
     Vector=&Multiply_Vector(
      A=&MakeVector(
       X=&Multiply_Float(
        A=&Multiply_Float(
         A=&UniformAOrBFloat(A=1, B=2.5, Evaluation Type=::NewEnumerator0, Recalculate Random Each Loop=true, Override Seed=false)
         B=&Sine(Normalized Angle=$Emitter.Age, Period=&RandomRangeFloat(Minimum=0.2, Maximum=1), Scale=2, Bias=1.5)
        )
        B=&Lerp_Float(
         A=&Power(a=&Multiply_Float(A=$System.MusicReactive, B=$NPC.FortniteNPC.EventAudio_Amplitude), b=1)
         B=1
         Alpha=$System.MusicReactivePreview
        )
       )
       Y=1
       Z=1
      )
      B=<1 1.5 1.5>
     )
     Float=0.025
    ))
   &ScaleColor(
    Scale Alpha=0.7
    Scale RGB=&Multiply_VectorByFloat(
     Vector=&MakeVectorFromLinearColorRGB(Color=&ColorFromCurve(
       ColorCurve=DataInterface() {propertyValues {
         AlphaCurve {keys []}
         BlueCurve {keys [<at 0 0.1859 RCIM_Cubic>,<at 0.5378 0.9502>,<at 0.7913 0>,<at 1 0.0208 RCIM_Cubic>]}
         GreenCurve {keys [<at 0 0.0762 RCIM_Cubic>,<at 0.5378 1>,<at 0.7913 0>,<at 1 0.0208 RCIM_Cubic>]}
         RedCurve {keys [<at 0 1 RCIM_Cubic>,<at 0.5378 0.009>,<at 0.7913 1>,<at 1 1 RCIM_Cubic>]}
        }}
       CurveIndex=&RandomRangeFloat(Minimum=0, Maximum=1)
      ))
     Float=0.5
    )
    ScaleA=true
    ScaleRGB=true
   ) {module=ScaleColor001}
   &ScaleColor(
    Scale Alpha=&Multiply_Float(
     A=&RandomRangeFloat(Minimum=0.5, Maximum=1)
     B=&MakeCustomFloatFromBool(Bool=&SetBoolByFloatComparison(A=$NPC.FortniteNPC.EventAudio_Amplitude, B=0), True Float=1, False Float=0)
    )
    Scale RGB=&Multiply_VectorByFloat(
     Vector=&MakeVectorFromLinearColorRGB(Color=&ColorFromCurve(
       ColorCurve=DataInterface() {propertyValues {
         AlphaCurve {keys []}
         BlueCurve {keys [<at 0 0.1859 RCIM_Cubic>,<at 0.5378 0.5832>,<at 0.7913 0>,<at 1 0.0208 RCIM_Cubic>]}
         GreenCurve {keys [<at 0 0.0762 RCIM_Cubic>,<at 0.5378 0.987>,<at 0.7913 0>,<at 1 0.0208 RCIM_Cubic>]}
         RedCurve {keys [<at 0 1 RCIM_Cubic>,<at 0.5378 2>,<at 0.7913 1>,<at 1 1 RCIM_Cubic>]}
        }}
       CurveIndex=&RandomRangeFloat(Minimum=0, Maximum=1)
      ))
     Float=0.5
    )
    ScaleA=true
    ScaleRGB=true
   )
  ]
  rendererClasses [&NiagaraMeshRendererProperties]
  renderers [{rendererClass=&NiagaraMeshRendererProperties, rendererIndex=0}]
 }
}
summary {
 systemName=NS_Emote_WarmShade_Loop
 emitters [
  {emitterName=SquareFront, rendererClasses [&NiagaraMeshRendererProperties]}
  {emitterName=SquareGlow, rendererClasses [&NiagaraMeshRendererProperties]}
  {emitterName=MeshRing, rendererClasses [&NiagaraMeshRendererProperties]}
  {emitterName=SideMesh, rendererClasses [&NiagaraMeshRendererProperties]}
  {bEnabled=false, emitterName=MainShape, rendererClasses [&NiagaraMeshRendererProperties]}
  {bEnabled=false, emitterName=OutLine, rendererClasses [&NiagaraMeshRendererProperties]}
  {emitterName=CenterRing, rendererClasses [&NiagaraMeshRendererProperties]}
  {emitterName=SquareRingR, rendererClasses [&NiagaraMeshRendererProperties]}
  {emitterName=SquareRingL, rendererClasses [&NiagaraMeshRendererProperties]}
  {bEnabled=false, emitterName=MainShape001, rendererClasses [&NiagaraMeshRendererProperties]}
  {bEnabled=false, emitterName=OutLine001, rendererClasses [&NiagaraMeshRendererProperties]}
  {emitterName=Glow, rendererClasses [&NiagaraSpriteRendererProperties, &NiagaraLightRendererProperties]}
  {emitterName=MainMeshShape, rendererClasses [&NiagaraMeshRendererProperties]}
 ]
 userVariables [
  User.bIsFrontEnd {type=NiagaraBool}
  User.bIsFrontEndPreview {type=NiagaraBool}
 ]
}
